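# Animation Blueprint for the Gravy Goblin weapon mesh, animating its ring being fired, grabbing an object and being caught.
# Each update copies HasRing and IsHoldingObject from the owning B_GravyGoblin_Athena into two booleans.
# A state machine leaves the ref-pose Idle when the ring is gone, loops after firing, optionally runs grab, hold and shoot, then catches when the ring returns.
# One-shot fire, grab and catch states advance automatically as their sequences end.
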AnimBlueprint /MotherGameplay/Items/GravyGoblin/Assets/Animation/GravyGoblin_AnimBP
target_skeleton=/MotherGameplay/Items/GravyGoblin/Assets/Meshes/Gravy_Goblin_Rifle_Skeleton
graphs [
 AnimGraph {nodes [
   AnimGraphNode_StateMachine_0 = Firing State() {is=title AnimGraphNode_StateMachine}
   Output Pose(Result=@AnimGraphNode_StateMachine_0) {layerGroup=None}
   Slot 'DefaultSlot'(Source=Local Space Ref Pose())
  ]}
 Firing State {
  entry=Idle
  path=AnimGraph.AnimGraphNode_StateMachine_0.Firing State
  ids=id
  nodes [
   AnimStateNode_0 = Idle()
   AnimStateNode_1 = Initial Fire() {is=title AnimStateNode}
   AnimStateNode_2 = Post-Fire Loop() {is=title AnimStateNode}
   AnimStateNode_3 = Catch() {is=title AnimStateNode}
   AnimStateNode_4 = Held Intro() {is=title AnimStateNode}
   AnimStateNode_5 = Held Loop() {is=title AnimStateNode}
   AnimStateNode_6 = Fire When Held() {is=title AnimStateNode}
   AnimStateTransitionNode_0 = AnimStateTransitionNode()
   AnimStateTransitionNode_1 = AnimStateTransitionNode() {bAutomaticRuleBasedOnSequencePlayerInState=true}
   AnimStateTransitionNode_2 = AnimStateTransitionNode()
   AnimStateTransitionNode_3 = AnimStateTransitionNode() {bAutomaticRuleBasedOnSequencePlayerInState=true}
   AnimStateTransitionNode_4 = AnimStateTransitionNode()
   AnimStateTransitionNode_5 = AnimStateTransitionNode() {bAutomaticRuleBasedOnSequencePlayerInState=true}
   AnimStateTransitionNode_6 = AnimStateTransitionNode()
   AnimStateTransitionNode_7 = AnimStateTransitionNode()
  ]
  links [
   {
    dst=Idle
    rule=AnimGraph.AnimGraphNode_StateMachine_0.Firing State.AnimStateTransitionNode_3.Transition
    src=Catch
    via=AnimStateTransitionNode_3
   }
   {
    dst=Catch
    rule=AnimGraph.AnimGraphNode_StateMachine_0.Firing State.AnimStateTransitionNode_7.Transition
    src=Fire When Held
    via=AnimStateTransitionNode_7
   }
   {
    dst=Held Loop
    rule=AnimGraph.AnimGraphNode_StateMachine_0.Firing State.AnimStateTransitionNode_5.Transition
    src=Held Intro
    via=AnimStateTransitionNode_5
   }
   {
    dst=Fire When Held
    rule=AnimGraph.AnimGraphNode_StateMachine_0.Firing State.AnimStateTransitionNode_6.Transition
    src=Held Loop
    via=AnimStateTransitionNode_6
   }
   {
    dst=Initial Fire
    rule=AnimGraph.AnimGraphNode_StateMachine_0.Firing State.AnimStateTransitionNode_0.Transition
    src=Idle
    via=AnimStateTransitionNode_0
   }
   {
    dst=Post-Fire Loop
    rule=AnimGraph.AnimGraphNode_StateMachine_0.Firing State.AnimStateTransitionNode_1.Transition
    src=Initial Fire
    via=AnimStateTransitionNode_1
   }
   {
    dst=Catch
    rule=AnimGraph.AnimGraphNode_StateMachine_0.Firing State.AnimStateTransitionNode_2.Transition
    src=Post-Fire Loop
    via=AnimStateTransitionNode_2
   }
   {
    dst=Held Intro
    rule=AnimGraph.AnimGraphNode_StateMachine_0.Firing State.AnimStateTransitionNode_4.Transition
    src=Post-Fire Loop
    via=AnimStateTransitionNode_4
   }
  ]
 }
 Idle {path=AnimGraph.AnimGraphNode_StateMachine_0.Firing State.AnimStateNode_0.Idle, nodes [Output Animation Pose(Result=Local Space Ref Pose())]}
 Initial Fire {
  path=AnimGraph.AnimGraphNode_StateMachine_0.Firing State.AnimStateNode_1.Initial Fire
  nodes [Output Animation Pose(Result=Sequence Player 'GravyGoblin_Fire_WEP'() {sequence=&/GravyGoblin_Fire_WEP})]
 }
 Post-Fire Loop {
  path=AnimGraph.AnimGraphNode_StateMachine_0.Firing State.AnimStateNode_2.Post-Fire Loop
  nodes [Output Animation Pose(Result=Sequence Player 'GravyGoblin_Fire_Loop_WEP'() {sequence=&/GravyGoblin_Fire_Loop_WEP})]
 }
 Catch {
  path=AnimGraph.AnimGraphNode_StateMachine_0.Firing State.AnimStateNode_3.Catch
  nodes [Output Animation Pose(Result=Sequence Player 'GravyGoblin_Catch_WEP'() {sequence=&/GravyGoblin_Catch_WEP})]
 }
 Held Intro {
  path=AnimGraph.AnimGraphNode_StateMachine_0.Firing State.AnimStateNode_4.Held Intro
  nodes [Output Animation Pose(Result=Sequence Player 'GravyGoblin_Grab_WEP'() {sequence=&/GravyGoblin_Grab_WEP})]
 }
 Held Loop {
  path=AnimGraph.AnimGraphNode_StateMachine_0.Firing State.AnimStateNode_5.Held Loop
  nodes [Output Animation Pose(Result=Sequence Player 'GravyGoblin_GrabLoop_WEP'() {sequence=&/GravyGoblin_GrabLoop_WEP})]
 }
 Fire When Held {
  path=AnimGraph.AnimGraphNode_StateMachine_0.Firing State.AnimStateNode_6.Fire When Held
  nodes [Output Animation Pose(Result=Sequence Player 'GravyGoblin_Grab_Shoot_WEP'() {sequence=&/GravyGoblin_Grab_Shoot_WEP})]
 }
 Transition {
  path=AnimGraph.AnimGraphNode_StateMachine_0.Firing State.AnimStateTransitionNode_0.Transition
  nodes [Result(bCanEnterTransition=NOT Boolean(A=Get bHasRing() -> (bHasRing:bool)))]
 }
 Transition {path=AnimGraph.AnimGraphNode_StateMachine_0.Firing State.AnimStateTransitionNode_1.Transition, nodes [Result()]}
 Transition {
  path=AnimGraph.AnimGraphNode_StateMachine_0.Firing State.AnimStateTransitionNode_2.Transition
  nodes [Result(bCanEnterTransition=Get bHasRing())]
 }
 Transition {path=AnimGraph.AnimGraphNode_StateMachine_0.Firing State.AnimStateTransitionNode_3.Transition, nodes [Result()]}
 Transition {
  path=AnimGraph.AnimGraphNode_StateMachine_0.Firing State.AnimStateTransitionNode_4.Transition
  nodes [Result(bCanEnterTransition=Get bItemHeld() -> (bItemHeld:bool))]
 }
 Transition {path=AnimGraph.AnimGraphNode_StateMachine_0.Firing State.AnimStateTransitionNode_5.Transition, nodes [Result()]}
 Transition {
  path=AnimGraph.AnimGraphNode_StateMachine_0.Firing State.AnimStateTransitionNode_6.Transition
  nodes [Result(bCanEnterTransition=NOT Boolean(A=Get bItemHeld()))]
 }
 Transition {
  path=AnimGraph.AnimGraphNode_StateMachine_0.Firing State.AnimStateTransitionNode_7.Transition
  nodes [Result(bCanEnterTransition=Get bHasRing())]
 }
 EventGraph {nodes [
   Event BlueprintUpdateAnimation()
   > n1 = Cast To B_GravyGoblin_Athena!(execute, Object=GetOwningActor()) -> (then, CastFailed, AsB Gravy Goblin Athena:B Gravy Goblin Athena Object Reference) {is=title K2Node_DynamicCast}
   > Set bHasRing(bHasRing:bool=Get HasRing!(self:B Gravy Goblin Athena Object Reference=@n1) -> (HasRing:bool)) -> (Output_Get:bool)
   > Set bItemHeld(bItemHeld:bool=Get IsHoldingObject!(self:B Gravy Goblin Athena Object Reference=@n1) -> (IsHoldingObject:bool)) -> (Output_Get:bool)
   TryGetPawnOwner()
  ]}
]
variables [
 bHasRing {type=boolean}
 bItemHeld {type=boolean}
]
