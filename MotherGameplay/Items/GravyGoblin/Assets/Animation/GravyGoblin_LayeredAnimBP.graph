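# Layered item animation blueprint for the Gravy Goblin weapon on the Fortnite player skeleton, animating firing, catching, grabbing and shooting while holding.
# Each update reads HasRing and IsHoldingObject from the current B_GravyGoblin_Athena weapon into two booleans.
# These booleans and sequence-end rules drive the FiringState machine, whose additive clips are applied onto the full-body pose in ItemFullBodyOverride.
# Among the Item Override layers, only ItemFullBodyOverride contains animation logic.
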AnimBlueprint /MotherGameplay/Items/GravyGoblin/Assets/Animation/GravyGoblin_LayeredAnimBP
target_skeleton=&Fortnite_M_Avg_Player_Skeleton
graphs [
 AnimGraph {nodes [Output Pose() {layerGroup=None}]}
 EventGraph {nodes [
   Event BlueprintUpdateAnimation()
   > n1 = Cast To B_GravyGoblin_Athena!(execute, Object=Get CurrentWeapon!(self:Fort Pawn Object Reference=TryGetFortPawn()) -> (CurrentWeapon:Fort Weapon Object Reference)) -> (then, CastFailed, AsB Gravy Goblin Athena:B Gravy Goblin Athena Object Reference) {is=title K2Node_DynamicCast}
   > Set bHasRing(bHasRing:bool=Get HasRing!(self:B Gravy Goblin Athena Object Reference=@n1) -> (HasRing:bool)) -> (Output_Get:bool)
   > Set bItemHeld(bItemHeld:bool=Get IsHoldingObject!(self:B Gravy Goblin Athena Object Reference=@n1) -> (IsHoldingObject:bool)) -> (Output_Get:bool)
   TryGetPawnOwner()
  ]}
 ItemAimOffset {nodes [
   Input Pose!() -> (Pose, AimOffsetAlpha:double, Yaw:double, Pitch:double) {name=InPose_AimOffset}
   Input Pose!() -> (Pose) {name=InPose_PreAimOffsets}
   Output Pose() {layerGroup=Item Override}
  ]}
 ItemFallAdditive {nodes [
   Input Pose() {name=InPoseItemFallAdditive}
   Output Pose() {layerGroup=Item Override}
  ]}
 ItemFallLandAdditive {nodes [
   Input Pose() {name=InPoseItemFallLandAdditive}
   Output Pose() {layerGroup=Item Override}
  ]}
 ItemFallLevel2Override {nodes [
   Input Pose() {name=InPose_FallLevel2}
   Output Pose() {layerGroup=Item Override}
  ]}
 ItemFinalPoseOverride {nodes [
   Input Pose() {name=InFinalPose}
   Output Pose() {layerGroup=Item Override}
  ]}
 ItemFullBodyAimOffsets {nodes [
   Input Pose!() -> (Pose, FullBody_AimOffsetAlpha:double, Aim_Yaw:double, Aim_Pitch:double) {name=InPoseFullBodyAimOffsets}
   Input Pose!() -> (Pose) {name=InPosePreFullBodyAimOffsets}
   Output Pose() {layerGroup=Item Override}
  ]}
 ItemFullBodyOverride {nodes [
   AnimGraphNode_StateMachine_0 = FiringState() {is=title AnimGraphNode_StateMachine}
   Output Pose(Result=Apply Additive(Base=Input Pose() {name=InPoseFullBody}, Additive=@AnimGraphNode_StateMachine_0) {alphaBoolBlend {blendOption=Linear}}) {layerGroup=Item Override}
  ]}
 FiringState {
  entry=Idle
  path=ItemFullBodyOverride.AnimGraphNode_StateMachine_0.FiringState
  ids=id
  nodes [
   AnimStateNode_0 = Idle()
   AnimStateNode_1 = Initial Fire() {is=title AnimStateNode}
   AnimStateNode_2 = Post-Fire Loop() {is=title AnimStateNode}
   AnimStateNode_3 = Catch() {is=title AnimStateNode}
   AnimStateNode_4 = Held Intro() {is=title AnimStateNode}
   AnimStateNode_5 = Held Loop() {is=title AnimStateNode}
   AnimStateNode_6 = Fire When Held() {is=title AnimStateNode}
   AnimStateTransitionNode_0 = AnimStateTransitionNode() {crossfadeDuration=0.05}
   AnimStateTransitionNode_1 = AnimStateTransitionNode() {bAutomaticRuleBasedOnSequencePlayerInState=true, crossfadeDuration=0.05}
   AnimStateTransitionNode_2 = AnimStateTransitionNode() {crossfadeDuration=0.05}
   AnimStateTransitionNode_3 = AnimStateTransitionNode() {bAutomaticRuleBasedOnSequencePlayerInState=true}
   AnimStateTransitionNode_4 = AnimStateTransitionNode() {crossfadeDuration=0.05}
   AnimStateTransitionNode_5 = AnimStateTransitionNode() {bAutomaticRuleBasedOnSequencePlayerInState=true, crossfadeDuration=0.05}
   AnimStateTransitionNode_6 = AnimStateTransitionNode() {crossfadeDuration=0}
   AnimStateTransitionNode_7 = AnimStateTransitionNode() {crossfadeDuration=0}
  ]
  links [
   {
    dst=Idle
    rule=ItemFullBodyOverride.AnimGraphNode_StateMachine_0.FiringState.AnimStateTransitionNode_3.AnimationTransitionGraph_0
    src=Catch
    via=AnimStateTransitionNode_3
   }
   {
    dst=Catch
    rule=ItemFullBodyOverride.AnimGraphNode_StateMachine_0.FiringState.AnimStateTransitionNode_7.Transition
    src=Fire When Held
    via=AnimStateTransitionNode_7
   }
   {
    dst=Held Loop
    rule=ItemFullBodyOverride.AnimGraphNode_StateMachine_0.FiringState.AnimStateTransitionNode_5.AnimationTransitionGraph_0
    src=Held Intro
    via=AnimStateTransitionNode_5
   }
   {
    dst=Fire When Held
    rule=ItemFullBodyOverride.AnimGraphNode_StateMachine_0.FiringState.AnimStateTransitionNode_6.AnimationTransitionGraph_0
    src=Held Loop
    via=AnimStateTransitionNode_6
   }
   {
    dst=Initial Fire
    rule=ItemFullBodyOverride.AnimGraphNode_StateMachine_0.FiringState.AnimStateTransitionNode_0.AnimationTransitionGraph_0
    src=Idle
    via=AnimStateTransitionNode_0
   }
   {
    dst=Post-Fire Loop
    rule=ItemFullBodyOverride.AnimGraphNode_StateMachine_0.FiringState.AnimStateTransitionNode_1.AnimationTransitionGraph_0
    src=Initial Fire
    via=AnimStateTransitionNode_1
   }
   {
    dst=Catch
    rule=ItemFullBodyOverride.AnimGraphNode_StateMachine_0.FiringState.AnimStateTransitionNode_2.AnimationTransitionGraph_0
    src=Post-Fire Loop
    via=AnimStateTransitionNode_2
   }
   {
    dst=Held Intro
    rule=ItemFullBodyOverride.AnimGraphNode_StateMachine_0.FiringState.AnimStateTransitionNode_4.AnimationTransitionGraph_0
    src=Post-Fire Loop
    via=AnimStateTransitionNode_4
   }
  ]
 }
 Idle {
  path=ItemFullBodyOverride.AnimGraphNode_StateMachine_0.FiringState.AnimStateNode_0.Idle
  nodes [Output Animation Pose(Result=Local Space Ref Pose())]
 }
 Initial Fire {
  path=ItemFullBodyOverride.AnimGraphNode_StateMachine_0.FiringState.AnimStateNode_1.Initial Fire
  nodes [Output Animation Pose(Result="Sequence Player (additive) 'GravyGoblin_Fire'"() {bLoopAnimation=false, sequence=&/GravyGoblin_Fire})]
 }
 Post-Fire Loop {
  path=ItemFullBodyOverride.AnimGraphNode_StateMachine_0.FiringState.AnimStateNode_2.Post-Fire Loop
  nodes [Output Animation Pose(Result="Sequence Player (additive) 'GravyGoblin_Fire_Loop'"() {sequence=&/GravyGoblin_Fire_Loop})]
 }
 Catch {
  path=ItemFullBodyOverride.AnimGraphNode_StateMachine_0.FiringState.AnimStateNode_3.Catch
  nodes [Output Animation Pose(Result="Sequence Player (additive) 'GravyGoblin_Catch'"() {sequence=&/GravyGoblin_Catch})]
 }
 Held Intro {
  path=ItemFullBodyOverride.AnimGraphNode_StateMachine_0.FiringState.AnimStateNode_4.Held Intro
  nodes [Output Animation Pose(Result="Sequence Player (additive) 'GravyGoblin_Grab'"() {sequence=&/GravyGoblin_Grab})]
 }
 Held Loop {
  path=ItemFullBodyOverride.AnimGraphNode_StateMachine_0.FiringState.AnimStateNode_5.Held Loop
  nodes [Output Animation Pose(Result="Sequence Player (additive) 'GravyGoblin_GrabLoop'"() {sequence=&/GravyGoblin_GrabLoop})]
 }
 Fire When Held {
  path=ItemFullBodyOverride.AnimGraphNode_StateMachine_0.FiringState.AnimStateNode_6.Fire When Held
  nodes [Output Animation Pose(Result="Sequence Player (additive) 'GravyGoblin_Grab_Shoot'"() {sequence=&/GravyGoblin_Grab_Shoot})]
 }
 AnimationTransitionGraph_0 {
  path=ItemFullBodyOverride.AnimGraphNode_StateMachine_0.FiringState.AnimStateTransitionNode_0.AnimationTransitionGraph_0
  nodes [Result(bCanEnterTransition=NOT Boolean(A=Get bHasRing() -> (bHasRing:bool)))]
 }
 AnimationTransitionGraph_0 {path=ItemFullBodyOverride.AnimGraphNode_StateMachine_0.FiringState.AnimStateTransitionNode_1.AnimationTransitionGraph_0, nodes [Result()]}
 AnimationTransitionGraph_0 {
  path=ItemFullBodyOverride.AnimGraphNode_StateMachine_0.FiringState.AnimStateTransitionNode_2.AnimationTransitionGraph_0
  nodes [Result(bCanEnterTransition=Get bHasRing())]
 }
 AnimationTransitionGraph_0 {path=ItemFullBodyOverride.AnimGraphNode_StateMachine_0.FiringState.AnimStateTransitionNode_3.AnimationTransitionGraph_0, nodes [Result()]}
 AnimationTransitionGraph_0 {
  path=ItemFullBodyOverride.AnimGraphNode_StateMachine_0.FiringState.AnimStateTransitionNode_4.AnimationTransitionGraph_0
  nodes [Result(bCanEnterTransition=Get bItemHeld() -> (bItemHeld:bool))]
 }
 AnimationTransitionGraph_0 {path=ItemFullBodyOverride.AnimGraphNode_StateMachine_0.FiringState.AnimStateTransitionNode_5.AnimationTransitionGraph_0, nodes [Result()]}
 AnimationTransitionGraph_0 {
  path=ItemFullBodyOverride.AnimGraphNode_StateMachine_0.FiringState.AnimStateTransitionNode_6.AnimationTransitionGraph_0
  nodes [Result(bCanEnterTransition=NOT Boolean(A=Get bItemHeld()))]
 }
 Transition {
  path=ItemFullBodyOverride.AnimGraphNode_StateMachine_0.FiringState.AnimStateTransitionNode_7.Transition
  nodes [Result(bCanEnterTransition=Get bHasRing())]
 }
 ItemIdleOverride {nodes [
   Input Pose() {name=InPose_Idle}
   Output Pose() {layerGroup=Item Override}
  ]}
 ItemJetPackJumpAdditive {nodes [Output Pose() {layerGroup=Item Override}]}
 ItemJetPackStartAdditive {nodes [Output Pose() {layerGroup=Item Override}]}
 ItemJumpLoopAdditive {nodes [
   Input Pose() {name=InPoseItemJumpLoopAdditive}
   Output Pose() {layerGroup=Item Override}
  ]}
 ItemJumpOverride {nodes [
   Input Pose() {name=InPoseJump}
   Output Pose() {layerGroup=Item Override}
  ]}
 ItemJumpUpAdditive {nodes [
   Input Pose() {name=InPoseItemJumpUpAdditive}
   Output Pose() {layerGroup=Item Override}
  ]}
 ItemPreAimOffsets {nodes [
   Input Pose() {name=InPosePreAimOffsets}
   Output Pose() {layerGroup=Item Override}
  ]}
 ItemPreIK {nodes [
   Input Pose() {name=InPose_PreIK}
   Output Pose() {layerGroup=Item Override}
  ]}
 ItemRightHandIK {nodes [
   Input Pose() {name=InPosePreRightHandIK}
   Input Pose() {name=InPosePostRightHandIK}
   Output Pose() {layerGroup=Item Override}
  ]}
 ItemSkeletalControl {nodes [
   Input Pose()
   Output Pose() {layerGroup=Item Override}
  ]}
 ItemSwimJumpFallLoop {nodes [Output Pose() {layerGroup=Item Override}]}
 ItemSwimJumpSurfaceEnd {nodes [Output Pose() {layerGroup=Item Override}]}
 ItemSwimJumpSurfaceLoop {nodes [Output Pose() {layerGroup=Item Override}]}
 ItemUpperBody {nodes [
   Input Pose!() -> (Pose, InputParam:Fort Anim Input Adjusted Aim Structure) {name=InPoseUpperBody}
   Output Pose() {layerGroup=Item Override}
  ]}
 ItemZipLineStartAdditive {nodes [Output Pose() {layerGroup=Item Override}]}
]
variables [
 bHasRing {type=boolean}
 bItemHeld {type=boolean}
]
