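# First-person fire sound cue for the Gravy Goblin launcher's disc shot.
# A Random node picks one of three equally weighted close-range waves, then a Modulator applies slight pitch variation.
# It routes through the first-person ranged-fire sound class and references first-person projectile attenuation and gunfire concurrency assets.
# The Random node is excluded from branch culling.
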SoundCue /MotherGameplay/Items/GravyGoblin/Assets/Audio/Launcher_Gravy_Goblin_Fire_Disc_1P_Cue
root=Modulator_1
nodes [
 {id=Modulator_1, children [Random_0], properties {pitchMax=1.1, pitchMin=0.9}}
 {id=Random_0, children [WavePlayer_1,WavePlayer_2,WavePlayer_3], properties {bShouldExcludeFromBranchCulling=true, weights [1,1,1]}}
 {id=WavePlayer_1, properties {soundWaveAssetPtr=&/Launcher_Gravy_Goblin_Fire_Disc_Close_01}}
 {id=WavePlayer_2, properties {soundWaveAssetPtr=&/Launcher_Gravy_Goblin_Fire_Disc_Close_02}}
 {id=WavePlayer_3, properties {soundWaveAssetPtr=&/Launcher_Gravy_Goblin_Fire_Disc_Close_03}}
]
properties {
 attenuationSettings=&Fort_Weapon_Projectile_FirstPerson_Athena_Att
 maxDistance=2097152
 soundClassObject=&Fort_WeaponRanged_Fire_1P_PSM
 volumeMultiplier=1
 attenuationOverrides {falloffDistance=5000, attenuationShapeExtents {x=1000}}
 concurrencyOverrides {bLimitToOwner=true, maxCount=3, resolutionRule=StopOldest}
 concurrencySet [&Weapon_Gunfire_FirstPerson_Conc]
}
