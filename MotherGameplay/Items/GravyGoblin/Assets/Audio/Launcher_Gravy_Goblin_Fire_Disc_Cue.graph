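# Sound cue for the Gravy Goblin launcher's disc fire sound, layered by distance into close, mid and distant.
# Each layer randomly picks one of three waves, then applies pitch modulation and its own projectile attenuation.
# Close plays through the third-person ranged fire sound class, while mid and distant are mixed at reduced volume under the distant fire class.
# Local and global gunfire concurrency sets are assigned.
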SoundCue /MotherGameplay/Items/GravyGoblin/Assets/Audio/Launcher_Gravy_Goblin_Fire_Disc_Cue
root=Mixer_2
nodes [
 {id=Attenuation_3, children [Modulator_6], properties {attenuationSettings=&Fort_Weapon_Projectile_Close_Athena_Att}}
 {
  id=Attenuation_4
  children [Modulator_5]
  properties {attenuationSettings=&Fort_Weapon_Projectile_Medium_Mono_Athena_Att}
 }
 {
  id=Attenuation_5
  children [Modulator_4]
  properties {attenuationSettings=&Fort_Weapon_Projectile_Distant_Mono_Athena_Att}
 }
 {id=Mixer_2, children [SoundClass_3,SoundClass_2], properties {inputVolume [1,1]}}
 {id=Mixer_4, children [Attenuation_4, Attenuation_5], properties {inputVolume [0.7,0.35]}}
 {id=Modulator_4, children [Random_1], properties {pitchMax=1.2}}
 {id=Modulator_5, children [Random_2], properties {pitchMax=1.2, volumeMax=1, volumeMin=1}}
 {id=Modulator_6, children [Random_0], properties {pitchMax=1.2}}
 {id=Random_0, children [WavePlayer_1,WavePlayer_2,WavePlayer_3], properties {bShouldExcludeFromBranchCulling=true, weights [1,1,1]}}
 {id=Random_1, children [WavePlayer_4,WavePlayer_6,WavePlayer_7], properties {bShouldExcludeFromBranchCulling=true, weights [1,1,1]}}
 {id=Random_2, children [WavePlayer_8, WavePlayer_9, WavePlayer_12], properties {bShouldExcludeFromBranchCulling=true, weights [1,1,1]}}
 {id=SoundClass_2, children [Mixer_4], properties {soundClassOverride=&Fort_WeaponRanged_Fire_Distant}}
 {
  id=SoundClass_3
  children [Attenuation_3]
  properties {soundClassOverride=&Fort_WeaponRanged_Fire_3P_PSM}
 }
 {id=WavePlayer_1, properties {soundWaveAssetPtr=&/Launcher_Gravy_Goblin_Fire_Disc_Close_01}}
 {id=WavePlayer_12, properties {soundWaveAssetPtr=&/Launcher_Gravy_Goblin_Fire_Disc_Mid_03}}
 {id=WavePlayer_2, properties {soundWaveAssetPtr=&/Launcher_Gravy_Goblin_Fire_Disc_Close_02}}
 {id=WavePlayer_3, properties {soundWaveAssetPtr=&/Launcher_Gravy_Goblin_Fire_Disc_Close_03}}
 {id=WavePlayer_4, properties {soundWaveAssetPtr=&/Launcher_Gravy_Goblin_Fire_Disc_Distant_01}}
 {id=WavePlayer_6, properties {soundWaveAssetPtr=&/Launcher_Gravy_Goblin_Fire_Disc_Distant_02}}
 {id=WavePlayer_7, properties {soundWaveAssetPtr=&/Launcher_Gravy_Goblin_Fire_Disc_Distant_03}}
 {id=WavePlayer_8, properties {soundWaveAssetPtr=&/Launcher_Gravy_Goblin_Fire_Disc_Mid_01}}
 {id=WavePlayer_9, properties {soundWaveAssetPtr=&/Launcher_Gravy_Goblin_Fire_Disc_Mid_02}}
]
properties {
 maxDistance=27500
 volumeMultiplier=1
 attenuationOverrides {falloffDistance=5000, attenuationShapeExtents {x=1000}}
 concurrencyOverrides {bLimitToOwner=true, maxCount=3, resolutionRule=StopOldest}
 concurrencySet [&Weapon_Gunfire_Local_Conc, &Weapon_Gunfire_Global_Conc]
}
