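# Sound cue for the Gravy Goblin launcher's first-person fire object sound.
# A random node picks one of three equally weighted close-range waves, is kept out of branch culling, and feeds a modulator.
# It uses first-person projectile attenuation, the first-person ranged weapon fire sound class and first-person gunfire concurrency.
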SoundCue /MotherGameplay/Items/GravyGoblin/Assets/Audio/Launcher_Gravy_Goblin_Fire_Object_1P_Cue
root=Modulator_1
nodes [
 {id=Modulator_1, children [Random_1]}
 {id=Random_1, children [WavePlayer_4,WavePlayer_5,WavePlayer_6], properties {bShouldExcludeFromBranchCulling=true, weights [1,1,1]}}
 {id=WavePlayer_4, properties {soundWaveAssetPtr=&/Launcher_Gravy_Goblin_Fire_Object_Close_01}}
 {id=WavePlayer_5, properties {soundWaveAssetPtr=&/Launcher_Gravy_Goblin_Fire_Object_Close_02}}
 {id=WavePlayer_6, properties {soundWaveAssetPtr=&/Launcher_Gravy_Goblin_Fire_Object_Close_03}}
]
properties {
 attenuationSettings=&Fort_Weapon_Projectile_FirstPerson_Athena_Att
 maxDistance=2097152
 soundClassObject=&Fort_WeaponRanged_Fire_1P_PSM
 volumeMultiplier=0.95
 concurrencySet [&Weapon_Gunfire_FirstPerson_Conc]
}
