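# Sound cue for the Gravy Goblin launcher firing an object, layered by listener distance.
# A root mixer blends a close layer in the third-person fire sound class with a mid and distant pair in the distant fire class, each layer using its own attenuation.
# Each layer picks randomly from three waves through a modulator for variation.
# Local and global gunfire concurrency sets limit simultaneous plays.
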SoundCue /MotherGameplay/Items/GravyGoblin/Assets/Audio/Launcher_Gravy_Goblin_Fire_Object_Cue
root=Mixer_0
nodes [
 {id=Attenuation_0, children [Modulator_2], properties {attenuationSettings=&Fort_Weapon_Projectile_Close_Athena_Att}}
 {
  id=Attenuation_1
  children [Modulator_1]
  properties {attenuationSettings=&Fort_Weapon_Projectile_Medium_Mono_Athena_Att}
 }
 {
  id=Attenuation_2
  children [Modulator_0]
  properties {attenuationSettings=&Fort_Weapon_Projectile_Distant_Mono_Athena_Att}
 }
 {id=Mixer_0, children [SoundClass_1,SoundClass_0], properties {inputVolume [1,1]}}
 {id=Mixer_3, children [Attenuation_1, Attenuation_2], properties {inputVolume [0.75,0.35]}}
 {id=Modulator_0, children [Random_4], properties {pitchMax=1.1}}
 {id=Modulator_1, children [Random_5], properties {pitchMax=1.1, volumeMax=1, volumeMin=1}}
 {id=Modulator_2, children [Random_3]}
 {id=Random_3, children [WavePlayer_12, WavePlayer_13, WavePlayer_14], properties {bShouldExcludeFromBranchCulling=true, weights [1,1,1]}}
 {id=Random_4, children [WavePlayer_15, WavePlayer_16, WavePlayer_17], properties {bShouldExcludeFromBranchCulling=true, weights [1,1,1]}}
 {id=Random_5, children [WavePlayer_18, WavePlayer_19, WavePlayer_20], properties {bShouldExcludeFromBranchCulling=true, weights [1,1,1]}}
 {id=SoundClass_0, children [Mixer_3], properties {soundClassOverride=&Fort_WeaponRanged_Fire_Distant}}
 {
  id=SoundClass_1
  children [Attenuation_0]
  properties {soundClassOverride=&Fort_WeaponRanged_Fire_3P_PSM}
 }
 {id=WavePlayer_12, properties {soundWaveAssetPtr=&/Launcher_Gravy_Goblin_Fire_Object_Close_01}}
 {id=WavePlayer_13, properties {soundWaveAssetPtr=&/Launcher_Gravy_Goblin_Fire_Object_Close_02}}
 {id=WavePlayer_14, properties {soundWaveAssetPtr=&/Launcher_Gravy_Goblin_Fire_Object_Close_03}}
 {id=WavePlayer_15, properties {soundWaveAssetPtr=&/Launcher_Gravy_Goblin_Fire_Object_Distant_01}}
 {id=WavePlayer_16, properties {soundWaveAssetPtr=&/Launcher_Gravy_Goblin_Fire_Object_Distant_02}}
 {id=WavePlayer_17, properties {soundWaveAssetPtr=&/Launcher_Gravy_Goblin_Fire_Object_Distant_03}}
 {id=WavePlayer_18, properties {soundWaveAssetPtr=&/Launcher_Gravy_Goblin_Fire_Object_Mid_01}}
 {id=WavePlayer_19, properties {soundWaveAssetPtr=&/Launcher_Gravy_Goblin_Fire_Object_Mid_02}}
 {id=WavePlayer_20, properties {soundWaveAssetPtr=&/Launcher_Gravy_Goblin_Fire_Object_Mid_03}}
]
properties {
 maxDistance=27500
 volumeMultiplier=0.85
 attenuationOverrides {falloffDistance=5000, attenuationShapeExtents {x=1000}}
 concurrencyOverrides {bLimitToOwner=true, maxCount=3, resolutionRule=StopOldest}
 concurrencySet [&Weapon_Gunfire_Local_Conc, &Weapon_Gunfire_Global_Conc]
}
