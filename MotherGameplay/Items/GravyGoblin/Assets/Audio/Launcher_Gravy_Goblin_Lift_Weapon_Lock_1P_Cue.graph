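# Sound cue for the first-person Gravy Goblin launcher's lift weapon lock sound.
# A Mixer layers one of six randomly chosen gun wobble variations with a slightly quieter activation sound, each routed through its own Modulator.
# It uses the first-person ranged weapon fire sound class, projectile attenuation and first-person gunfire concurrency.
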SoundCue /MotherGameplay/Items/GravyGoblin/Assets/Audio/Launcher_Gravy_Goblin_Lift_Weapon_Lock_1P_Cue
root=Mixer_0
nodes [
 {id=Mixer_0, children [Modulator_0,Modulator_1], properties {inputVolume [1,0.85]}}
 {id=Modulator_0, children [Random_0]}
 {id=Modulator_1, children [WavePlayer_7]}
 {id=Random_0, children [WavePlayer_0,WavePlayer_1,WavePlayer_2,WavePlayer_3,WavePlayer_4,WavePlayer_6], properties {weights [1,1,1,1,1,1]}}
 {id=WavePlayer_0, properties {soundWaveAssetPtr=&/Launcher_Gravy_Goblin_Lift_Gun_Wobble_01}}
 {id=WavePlayer_1, properties {soundWaveAssetPtr=&/Launcher_Gravy_Goblin_Lift_Gun_Wobble_02}}
 {id=WavePlayer_2, properties {soundWaveAssetPtr=&/Launcher_Gravy_Goblin_Lift_Gun_Wobble_03}}
 {id=WavePlayer_3, properties {soundWaveAssetPtr=&/Launcher_Gravy_Goblin_Lift_Gun_Wobble_04}}
 {id=WavePlayer_4, properties {soundWaveAssetPtr=&/Launcher_Gravy_Goblin_Lift_Gun_Wobble_05}}
 {id=WavePlayer_6, properties {soundWaveAssetPtr=&/Launcher_Gravy_Goblin_Lift_Gun_Wobble_06}}
 {id=WavePlayer_7, properties {soundWaveAssetPtr=&/Launcher_Gravy_Goblin_Lift_Gun_Activate_01}}
]
properties {
 attenuationSettings=&Fort_Weapon_Projectile_FirstPerson_Athena_Att
 maxDistance=2097152
 soundClassObject=&Fort_WeaponRanged_Fire_1P_PSM
 volumeMultiplier=0.65
 attenuationOverrides {falloffDistance=5000, attenuationShapeExtents {x=1000}}
 concurrencyOverrides {bLimitToOwner=true, maxCount=3, resolutionRule=StopOldest}
 concurrencySet [&Weapon_Gunfire_FirstPerson_Conc]
}
