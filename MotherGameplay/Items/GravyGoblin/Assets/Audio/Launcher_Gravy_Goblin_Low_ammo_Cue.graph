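# Sound cue for the Gravy Goblin launcher's low-ammo warning, though it has no root node or wave nodes and so plays nothing as shown.
# It routes to the ranged-weapon low-ammo degraded PSM sound class and sets a very large max distance.
# A concurrency override blocks new instances once the limit is reached rather than stopping playing ones.
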SoundCue /MotherGameplay/Items/GravyGoblin/Assets/Audio/Launcher_Gravy_Goblin_Low_ammo_Cue
root=null
nodes []
properties {
 bOverrideConcurrency=true
 maxDistance=2097152
 soundClassObject=/Game/Sounds/FortSoundClassesAndModes/Fort_WeaponRanged_LowAmmo_Degraded_PSM
 concurrencyOverrides {resolutionRule=PreventNew}
}
