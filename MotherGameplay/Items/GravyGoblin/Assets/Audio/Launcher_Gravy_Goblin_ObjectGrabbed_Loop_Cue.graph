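# Sound cue for the Gravy Goblin launcher's looping audio while an object is grabbed.
# An Enveloper fades in two mixed loops, projectile idle and attachment line, layered with one of three random, modulated loop-start sounds.
# Continuous modulators drive both loops' pitch and volume from Velocity and ADS parameters.
# Rising ADS boosts the idle loop's volume and ducks the line loop, while both rise in pitch.
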SoundCue /MotherGameplay/Items/GravyGoblin/Assets/Audio/Launcher_Gravy_Goblin_ObjectGrabbed_Loop_Cue
root=Mixer_2
nodes [
 {
  id=Enveloper_1
  children [Mixer_1]
  properties {pitchCurve {editorCurveData {keys [{time=0}]}}, volumeCurve {editorCurveData {keys [<at 0 0 RCIM_Cubic>,<at 1 1>]}}}
 }
 {id=Mixer_1, children [ModulatorContinuous_2, ModulatorContinuous_5], properties {inputVolume [3,1]}}
 {id=Mixer_2, children [Enveloper_1,Modulator_0], properties {inputVolume [1,0.75]}}
 {
  id=ModulatorContinuous_2
  children [ModulatorContinuous_4]
  properties {
   pitchModulationParams {default=0, maxOutput=1.5, minOutput=1, parameterName=ADS}
   volumeModulationParams {default=0, maxOutput=1.5, minOutput=0.75, parameterName=ADS}
  }
 }
 {
  id=ModulatorContinuous_3
  children [WavePlayer_5]
  properties {
   pitchModulationParams {default=0, maxOutput=1.5, minOutput=1, parameterName=Velocity}
   volumeModulationParams {default=0, minOutput=0.2, parameterName=Velocity}
  }
 }
 {
  id=ModulatorContinuous_4
  children [WavePlayer_3]
  properties {
   pitchModulationParams {default=0, maxOutput=1.25, minOutput=1, parameterName=Velocity}
   volumeModulationParams {default=0, minOutput=0.75, parameterName=Velocity}
  }
 }
 {
  id=ModulatorContinuous_5
  children [ModulatorContinuous_3]
  properties {
   pitchModulationParams {default=0, maxOutput=1.25, minOutput=1, parameterName=ADS}
   volumeModulationParams {default=0, maxOutput=0.2, minOutput=1, parameterName=ADS}
  }
 }
 {id=Modulator_0, children [Random_0]}
 {id=Random_0, children [WavePlayer_0,WavePlayer_1,WavePlayer_2], properties {bShouldExcludeFromBranchCulling=true, weights [1,1,1]}}
 {
  id=WavePlayer_0
  properties {soundWaveAssetPtr=&/Launcher_Gravy_Goblin_Attachment_Line_Loop_Start_01}
 }
 {
  id=WavePlayer_1
  properties {soundWaveAssetPtr=&/Launcher_Gravy_Goblin_Attachment_Line_Loop_Start_02}
 }
 {
  id=WavePlayer_2
  properties {soundWaveAssetPtr=&/Launcher_Gravy_Goblin_Attachment_Line_Loop_Start_03}
 }
 {
  id=WavePlayer_3
  properties {bLooping=true, soundWaveAssetPtr=&/Launcher_Gravy_Goblin_Prj_Idle_Loop}
 }
 {
  id=WavePlayer_5
  properties {bLooping=true, soundWaveAssetPtr=&/Launcher_Gravy_Goblin_Attachment_Line_Loop}
 }
]
properties {
 attenuationSettings=&/Fort_Weapon_Launcher_Gravy_Attachment_Line_Att
 maxDistance=9450
 soundClassObject=&Fort_WeaponRanged_Fire_1P_PSM
 volumeMultiplier=0.65
 attenuationOverrides {falloffDistance=5000, attenuationShapeExtents {x=1000}}
 concurrencyOverrides {maxCount=3}
 concurrencySet [&Weapon_Gunfire_FirstPerson_Conc]
}
