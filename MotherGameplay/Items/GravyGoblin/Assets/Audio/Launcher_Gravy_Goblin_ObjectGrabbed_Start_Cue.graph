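# Sound cue for the start of the Gravy Goblin launcher grabbing an object.
# A WasCaught branch plays a random large in-air whip when true, otherwise a Size switch selects small or regular object-lift sounds.
# Each lift set uses audio quality levels to pick between a larger and a smaller random pool, with modulators varying every path.
# Attenuation and first-person gunfire concurrency govern playback.
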SoundCue /MotherGameplay/Items/GravyGoblin/Assets/Audio/Launcher_Gravy_Goblin_ObjectGrabbed_Start_Cue
root=Branch_0
nodes [
 {id=Branch_0, children [Mixer_2,Switch_0,Switch_0], properties {boolParameterName=WasCaught}}
 {id=Mixer_0, children [Modulator_3], properties {inputVolume [1]}}
 {id=Mixer_1, children [Modulator_0], properties {inputVolume [1]}}
 {id=Mixer_2, children [Modulator_2], properties {inputVolume [1]}}
 {id=Modulator_0, children [QualityLevel_0]}
 {id=Modulator_2, children [Random_3]}
 {id=Modulator_3, children [QualityLevel_1]}
 {id=QualityLevel_0, children [Random_0,Random_0,Random_1,Random_1]}
 {id=QualityLevel_1, children [Random_2,Random_2,Random_5,Random_5]}
 {
  id=Random_0
  children [WavePlayer_5, WavePlayer_0, WavePlayer_1, WavePlayer_2, WavePlayer_3, WavePlayer_15]
  properties {bShouldExcludeFromBranchCulling=true, weights [1,1,1,1,1,1]}
 }
 {id=Random_1, children [WavePlayer_16, WavePlayer_17, WavePlayer_18], properties {bShouldExcludeFromBranchCulling=true, weights [1,1,1]}}
 {
  id=Random_2
  children [WavePlayer_19, WavePlayer_20, WavePlayer_21, WavePlayer_22, WavePlayer_23]
  properties {bShouldExcludeFromBranchCulling=true, weights [1,1,1,1,1]}
 }
 {id=Random_3, children [WavePlayer_6,WavePlayer_7,WavePlayer_8,WavePlayer_9], properties {weights [1,1,1,1]}}
 {id=Random_5, children [WavePlayer_24, WavePlayer_25, WavePlayer_26], properties {bShouldExcludeFromBranchCulling=true, weights [1,1,1]}}
 {id=Switch_0, children [Mixer_0,Mixer_0,Mixer_1,Mixer_1], properties {intParameterName=Size}}
 {id=WavePlayer_0, properties {soundWaveAssetPtr=&/Launcher_Gravy_Goblin_Lift_Object_02}}
 {id=WavePlayer_1, properties {soundWaveAssetPtr=&/Launcher_Gravy_Goblin_Lift_Object_03}}
 {id=WavePlayer_15, properties {soundWaveAssetPtr=&/Launcher_Gravy_Goblin_Lift_Object_06}}
 {id=WavePlayer_16, properties {soundWaveAssetPtr=&/Launcher_Gravy_Goblin_Lift_Object_01}}
 {id=WavePlayer_17, properties {soundWaveAssetPtr=&Launcher_Gravy_Goblin_Lift_Object_02}}
 {id=WavePlayer_18, properties {soundWaveAssetPtr=&Launcher_Gravy_Goblin_Lift_Object_03}}
 {id=WavePlayer_19, properties {soundWaveAssetPtr=&/Launcher_Gravy_Goblin_Lift_Object_Sml_01}}
 {id=WavePlayer_2, properties {soundWaveAssetPtr=&/Launcher_Gravy_Goblin_Lift_Object_04}}
 {id=WavePlayer_20, properties {soundWaveAssetPtr=&/Launcher_Gravy_Goblin_Lift_Object_Sml_02}}
 {id=WavePlayer_21, properties {soundWaveAssetPtr=&/Launcher_Gravy_Goblin_Lift_Object_Sml_03}}
 {id=WavePlayer_22, properties {soundWaveAssetPtr=&/Launcher_Gravy_Goblin_Lift_Object_Sml_04}}
 {id=WavePlayer_23, properties {soundWaveAssetPtr=&/Launcher_Gravy_Goblin_Lift_Object_Sml_05}}
 {id=WavePlayer_24, properties {soundWaveAssetPtr=&Launcher_Gravy_Goblin_Lift_Object_Sml_01}}
 {id=WavePlayer_25, properties {soundWaveAssetPtr=&Launcher_Gravy_Goblin_Lift_Object_Sml_02}}
 {id=WavePlayer_26, properties {soundWaveAssetPtr=&Launcher_Gravy_Goblin_Lift_Object_Sml_03}}
 {id=WavePlayer_3, properties {soundWaveAssetPtr=&/Launcher_Gravy_Goblin_Lift_Object_05}}
 {id=WavePlayer_5, properties {soundWaveAssetPtr=&Launcher_Gravy_Goblin_Lift_Object_01}}
 {
  id=WavePlayer_6
  properties {soundWaveAssetPtr=&/Launcher_Gravy_Goblin_Prj_Object_InAir_Whip_Large_01}
 }
 {
  id=WavePlayer_7
  properties {soundWaveAssetPtr=&/Launcher_Gravy_Goblin_Prj_Object_InAir_Whip_Large_02}
 }
 {
  id=WavePlayer_8
  properties {soundWaveAssetPtr=&/Launcher_Gravy_Goblin_Prj_Object_InAir_Whip_Large_03}
 }
 {
  id=WavePlayer_9
  properties {soundWaveAssetPtr=&/Launcher_Gravy_Goblin_Prj_Object_InAir_Whip_Large_04}
 }
]
properties {
 attenuationSettings=&/Fort_Weapon_Launcher_Gravy_Object_Wiggle_Att
 maxDistance=12200
 soundClassObject=&Fort_WeaponRanged_Fire_1P_PSM
 volumeMultiplier=0.8
 attenuationOverrides {falloffDistance=5000, attenuationShapeExtents {x=1000}}
 concurrencyOverrides {maxCount=3}
 concurrencySet [&Weapon_Gunfire_FirstPerson_Conc]
}
