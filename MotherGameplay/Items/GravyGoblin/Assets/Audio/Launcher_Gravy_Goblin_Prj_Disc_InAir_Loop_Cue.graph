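# Sound cue for the in-air sound of the Gravy Goblin launcher's disc projectile.
# Mixes a looping in-air wave with a quieter whoosh, picked at random from three variants with slight pitch variation.
# A continuous modulator wraps the mix, driven by a Distance parameter that lowers pitch as its value rises.
# Uses projectile attenuation and overrides concurrency to one instance per owner, stopping the oldest.
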SoundCue /MotherGameplay/Items/GravyGoblin/Assets/Audio/Launcher_Gravy_Goblin_Prj_Disc_InAir_Loop_Cue
root=ModulatorContinuous_0
nodes [
 {id=Mixer_0, children [WavePlayer_5,Modulator_0], properties {inputVolume [0.85,0.25]}}
 {
  id=ModulatorContinuous_0
  children [Mixer_0]
  properties {
   pitchModulationParams {default=0, maxOutput=0.7, minOutput=1.5, parameterName=Distance}
   volumeModulationParams {default=0, minOutput=1, parameterName=Distance}
  }
 }
 {id=Modulator_0, children [Random_0], properties {pitchMax=1.1, pitchMin=0.9}}
 {id=Random_0, children [WavePlayer_1,WavePlayer_2,WavePlayer_3], properties {bShouldExcludeFromBranchCulling=true, weights [1,1,1]}}
 {id=WavePlayer_1, properties {soundWaveAssetPtr=&/Launcher_Gravy_Goblin_Prj_Disc_Whsh_01}}
 {id=WavePlayer_2, properties {soundWaveAssetPtr=&/Launcher_Gravy_Goblin_Prj_Disc_Whsh_02}}
 {id=WavePlayer_3, properties {soundWaveAssetPtr=&/Launcher_Gravy_Goblin_Prj_Disc_Whsh_03}}
 {
  id=WavePlayer_5
  properties {bLooping=true, soundWaveAssetPtr=&/Launcher_Gravy_Goblin_Prj_Disc_InAir_Loop}
 }
]
properties {
 attenuationSettings=&/Fort_Weapon_Launcher_Gravy_Disc_Prj_Att
 bOverrideConcurrency=true
 maxDistance=14750
 soundClassObject=&Fort_WeaponRanged_Fire_3P_PSM
 volumeMultiplier=1
 attenuationOverrides {falloffDistance=5000, attenuationShapeExtents {x=1000}}
 concurrencyOverrides {bLimitToOwner=true, maxCount=1, resolutionRule=StopOldest}
 concurrencySet [&Weapon_Gunfire_Local_Conc, &Weapon_Gunfire_Global_Conc]
}
