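# Sound cue for the stop sound of the Gravy Goblin launcher's in-air disc projectile loop.
# The first two quality levels randomly pick one of three equally weighted waves with pitch variation; the last two use only the first wave, also pitch-modulated.
# It uses projectile-stop attenuation and a third-person ranged weapon fire sound class.
# A concurrency override limits simultaneous plays, superseding the listed gunfire concurrency sets.
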SoundCue /MotherGameplay/Items/GravyGoblin/Assets/Audio/Launcher_Gravy_Goblin_Prj_Disc_InAir_Loop_Stop_Cue
root=QualityLevel_0
nodes [
 {id=Modulator_0, children [WavePlayer_0], properties {pitchMax=1.3, pitchMin=0.9}}
 {id=Modulator_1, children [Random_0], properties {pitchMax=1.25}}
 {id=QualityLevel_0, children [Modulator_1,Modulator_1,Modulator_0,Modulator_0]}
 {id=Random_0, children [WavePlayer_3,WavePlayer_1,WavePlayer_2], properties {bShouldExcludeFromBranchCulling=true, weights [1,1,1]}}
 {
  id=WavePlayer_0
  properties {soundWaveAssetPtr=&/Launcher_Gravy_Goblin_Prj_Disc_InAir_Loop_Stop_01}
 }
 {
  id=WavePlayer_1
  properties {soundWaveAssetPtr=&/Launcher_Gravy_Goblin_Prj_Disc_InAir_Loop_Stop_02}
 }
 {
  id=WavePlayer_2
  properties {soundWaveAssetPtr=&/Launcher_Gravy_Goblin_Prj_Disc_InAir_Loop_Stop_03}
 }
 {
  id=WavePlayer_3
  properties {soundWaveAssetPtr=&Launcher_Gravy_Goblin_Prj_Disc_InAir_Loop_Stop_01}
 }
]
properties {
 attenuationSettings=&/Fort_Weapon_Launcher_Gravy_Disc_Prj_Stop_Att
 bOverrideConcurrency=true
 maxDistance=12750
 soundClassObject=&Fort_WeaponRanged_Fire_3P_PSM
 volumeMultiplier=0.6
 attenuationOverrides {falloffDistance=5000, attenuationShapeExtents {x=1000}}
 concurrencyOverrides {maxCount=2}
 concurrencySet [&Weapon_Gunfire_Local_Conc, &Weapon_Gunfire_Global_Conc]
}
