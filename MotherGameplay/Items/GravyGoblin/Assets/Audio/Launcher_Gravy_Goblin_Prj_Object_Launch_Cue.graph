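# Sound cue for the launch of a Gravy Goblin launcher projectile object, in the ranged weapon sound class with launch attenuation.
# After a short fixed delay, a mixer layers a random large in-air whip with a random lift-gun wobble.
# A quality-level node picks the wobble pool, either six variations or a three-sound subset, and all random nodes are excluded from branch culling.
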SoundCue /MotherGameplay/Items/GravyGoblin/Assets/Audio/Launcher_Gravy_Goblin_Prj_Object_Launch_Cue
root=Delay_1
nodes [
 {id=Delay_1, children [Mixer_0], properties {delayMax=0.35, delayMin=0.35}}
 {id=Mixer_0, children [Random_1, QualityLevel_0], properties {inputVolume [1,1]}}
 {id=QualityLevel_0, children [Random_0,Random_0,Random_2,Random_2]}
 {
  id=Random_0
  children [WavePlayer_1,WavePlayer_2,WavePlayer_3,WavePlayer_4,WavePlayer_8,WavePlayer_9]
  properties {bShouldExcludeFromBranchCulling=true, weights [1,1,1,1,1,1]}
 }
 {
  id=Random_1
  children [WavePlayer_0,WavePlayer_5,WavePlayer_6,WavePlayer_7]
  properties {bShouldExcludeFromBranchCulling=true, weights [1,1,1,1]}
 }
 {id=Random_2, children [WavePlayer_10, WavePlayer_11, WavePlayer_12], properties {bShouldExcludeFromBranchCulling=true, weights [1,1,1]}}
 {
  id=WavePlayer_0
  properties {soundWaveAssetPtr=&/Launcher_Gravy_Goblin_Prj_Object_InAir_Whip_Large_01}
 }
 {id=WavePlayer_1, properties {soundWaveAssetPtr=&/Launcher_Gravy_Goblin_Lift_Gun_Wobble_06}}
 {id=WavePlayer_10, properties {soundWaveAssetPtr=&Launcher_Gravy_Goblin_Lift_Gun_Wobble_06}}
 {id=WavePlayer_11, properties {soundWaveAssetPtr=&/Launcher_Gravy_Goblin_Lift_Gun_Wobble_01}}
 {id=WavePlayer_12, properties {soundWaveAssetPtr=&/Launcher_Gravy_Goblin_Lift_Gun_Wobble_02}}
 {id=WavePlayer_2, properties {soundWaveAssetPtr=&Launcher_Gravy_Goblin_Lift_Gun_Wobble_01}}
 {id=WavePlayer_3, properties {soundWaveAssetPtr=&Launcher_Gravy_Goblin_Lift_Gun_Wobble_02}}
 {id=WavePlayer_4, properties {soundWaveAssetPtr=&/Launcher_Gravy_Goblin_Lift_Gun_Wobble_03}}
 {
  id=WavePlayer_5
  properties {soundWaveAssetPtr=&/Launcher_Gravy_Goblin_Prj_Object_InAir_Whip_Large_02}
 }
 {
  id=WavePlayer_6
  properties {soundWaveAssetPtr=&/Launcher_Gravy_Goblin_Prj_Object_InAir_Whip_Large_03}
 }
 {
  id=WavePlayer_7
  properties {soundWaveAssetPtr=&/Launcher_Gravy_Goblin_Prj_Object_InAir_Whip_Large_04}
 }
 {id=WavePlayer_8, properties {soundWaveAssetPtr=&/Launcher_Gravy_Goblin_Lift_Gun_Wobble_04}}
 {id=WavePlayer_9, properties {soundWaveAssetPtr=&/Launcher_Gravy_Goblin_Lift_Gun_Wobble_05}}
]
properties {
 attenuationSettings=&/Fort_Weapon_Launcher_Gravy_Object_Prj_Launch_Att
 bDebug=true
 maxDistance=16500
 soundClassObject=&Fort_WeaponRanged
 volumeMultiplier=0.7
 attenuationOverrides {
  bAttenuate=false
  bEnableReverbSend=false
  bUseComplexCollisionForOcclusion=true
  falloffDistance=0
  occlusionVolumeAttenuation=0
  stereoSpread=0
  attenuationShapeExtents {x=50000}
 }
}
