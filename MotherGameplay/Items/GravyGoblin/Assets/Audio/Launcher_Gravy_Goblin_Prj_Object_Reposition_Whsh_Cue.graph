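# Sound cue for the whoosh of an object being repositioned by the Gravy Goblin launcher, reusing its Hold whoosh waves.
# An integer Size switch picks a small or large branch, each randomizing volume and raising pitch by a random amount, with a wider pitch range for large.
# Quality level nodes then select either a fuller random wave pool or a separate three-wave pool.
# Attenuation and gunfire concurrency settings govern playback.
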SoundCue /MotherGameplay/Items/GravyGoblin/Assets/Audio/Launcher_Gravy_Goblin_Prj_Object_Reposition_Whsh_Cue
root=Switch_0
nodes [
 {id=Modulator_0, children [QualityLevel_0], properties {pitchMax=1.5, pitchMin=1, volumeMax=1, volumeMin=0.5}}
 {id=Modulator_1, children [QualityLevel_1], properties {pitchMax=1.2, pitchMin=1, volumeMax=1, volumeMin=0.5}}
 {id=QualityLevel_0, children [Random_0,Random_0,Random_1,Random_1]}
 {id=QualityLevel_1, children [Random_5,Random_5,Random_6,Random_6]}
 {
  id=Random_0
  children [WavePlayer_0,WavePlayer_1,WavePlayer_2,WavePlayer_3,WavePlayer_4,WavePlayer_6]
  properties {bShouldExcludeFromBranchCulling=true, weights [1,1,1,1,1,1]}
 }
 {id=Random_1, children [WavePlayer_7,WavePlayer_8,WavePlayer_9], properties {bShouldExcludeFromBranchCulling=true, weights [1,1,1]}}
 {
  id=Random_5
  children [WavePlayer_25, WavePlayer_26, WavePlayer_27, WavePlayer_28, WavePlayer_29]
  properties {bShouldExcludeFromBranchCulling=true, weights [1,1,1,1,1]}
 }
 {id=Random_6, children [WavePlayer_30, WavePlayer_31, WavePlayer_32], properties {bShouldExcludeFromBranchCulling=true, weights [1,1,1]}}
 {id=Switch_0, children [Modulator_1,Modulator_1,Modulator_0,Modulator_0], properties {intParameterName=Size}}
 {
  id=WavePlayer_0
  properties {soundWaveAssetPtr=&/Launcher_Gravy_Goblin_Prj_Object_Hold_Lrg_Whsh_01}
 }
 {
  id=WavePlayer_1
  properties {soundWaveAssetPtr=&/Launcher_Gravy_Goblin_Prj_Object_Hold_Lrg_Whsh_02}
 }
 {
  id=WavePlayer_2
  properties {soundWaveAssetPtr=&/Launcher_Gravy_Goblin_Prj_Object_Hold_Lrg_Whsh_03}
 }
 {
  id=WavePlayer_25
  properties {soundWaveAssetPtr=&/Launcher_Gravy_Goblin_Prj_Object_Hold_Sml_Whsh_01}
 }
 {
  id=WavePlayer_26
  properties {soundWaveAssetPtr=&/Launcher_Gravy_Goblin_Prj_Object_Hold_Sml_Whsh_02}
 }
 {
  id=WavePlayer_27
  properties {soundWaveAssetPtr=&/Launcher_Gravy_Goblin_Prj_Object_Hold_Sml_Whsh_03}
 }
 {
  id=WavePlayer_28
  properties {soundWaveAssetPtr=&/Launcher_Gravy_Goblin_Prj_Object_Hold_Sml_Whsh_04}
 }
 {
  id=WavePlayer_29
  properties {soundWaveAssetPtr=&/Launcher_Gravy_Goblin_Prj_Object_Hold_Sml_Whsh_05}
 }
 {
  id=WavePlayer_3
  properties {soundWaveAssetPtr=&/Launcher_Gravy_Goblin_Prj_Object_Hold_Lrg_Whsh_04}
 }
 {
  id=WavePlayer_30
  properties {soundWaveAssetPtr=&Launcher_Gravy_Goblin_Prj_Object_Hold_Sml_Whsh_01}
 }
 {
  id=WavePlayer_31
  properties {soundWaveAssetPtr=&Launcher_Gravy_Goblin_Prj_Object_Hold_Sml_Whsh_02}
 }
 {
  id=WavePlayer_32
  properties {soundWaveAssetPtr=&Launcher_Gravy_Goblin_Prj_Object_Hold_Sml_Whsh_03}
 }
 {
  id=WavePlayer_4
  properties {soundWaveAssetPtr=&/Launcher_Gravy_Goblin_Prj_Object_Hold_Lrg_Whsh_05}
 }
 {
  id=WavePlayer_6
  properties {soundWaveAssetPtr=&/Launcher_Gravy_Goblin_Prj_Object_Hold_Lrg_Whsh_06}
 }
 {
  id=WavePlayer_7
  properties {soundWaveAssetPtr=&Launcher_Gravy_Goblin_Prj_Object_Hold_Lrg_Whsh_01}
 }
 {
  id=WavePlayer_8
  properties {soundWaveAssetPtr=&Launcher_Gravy_Goblin_Prj_Object_Hold_Lrg_Whsh_02}
 }
 {
  id=WavePlayer_9
  properties {soundWaveAssetPtr=&Launcher_Gravy_Goblin_Prj_Object_Hold_Lrg_Whsh_03}
 }
]
properties {
 attenuationSettings=&/Fort_Weapon_Launcher_Gravy_Object_Prj_Hold_Whsh_Att
 bOverrideConcurrency=true
 maxDistance=9450
 soundClassObject=&Fort_WeaponRanged_Fire_3P_PSM
 attenuationOverrides {falloffDistance=5000, attenuationShapeExtents {x=1000}}
 concurrencyOverrides {maxCount=3}
 concurrencySet [&Weapon_Gunfire_Local_Conc, &Weapon_Gunfire_Global_Conc]
}
