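# Sound cue for the Gravy Goblin launcher's reload-end foley sound.
# A single wave player feeds a modulator that randomizes pitch slightly while holding volume constant.
# It routes through the weapon foley sound class with shared reload attenuation, a reduced volume multiplier, a concurrency override, and debug enabled.
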SoundCue /MotherGameplay/Items/GravyGoblin/Assets/Audio/Launcher_Gravy_Goblin_Reload_End_Cue
root=Modulator_14
nodes [
 {id=Modulator_14, children [WavePlayer_0], properties {pitchMax=1.1, volumeMax=1, volumeMin=1}}
 {id=WavePlayer_0, properties {soundWaveAssetPtr=&/Launcher_Gravy_Goblin_Reload_End_01}}
]
properties {
 attenuationSettings=&Fort_Weapon_Reload_Att
 bDebug=true
 bOverrideConcurrency=true
 maxDistance=2250
 soundClassObject=&Fort_WeaponFoley
 volumeMultiplier=0.601
}
