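# Sound cue for the start of the Gravy Goblin launcher reload, classed as weapon foley at roughly half volume.
# It references the shared Fortnite weapon reload attenuation and overrides concurrency to a single instance, resolved by stopping the farthest and then preventing new ones.
# The node graph is empty with no root, so the cue references no sound wave and has nothing to play.
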SoundCue /MotherGameplay/Items/GravyGoblin/Assets/Audio/Launcher_Gravy_Goblin_Reload_Start_Cue
root=null
nodes []
properties {
 attenuationSettings=&Fort_Weapon_Reload_Att
 bOverrideConcurrency=true
 maxDistance=2100
 soundClassObject=&Fort_WeaponFoley
 volumeMultiplier=0.501
 attenuationOverrides {falloffDistance=5000, attenuationShapeExtents {x=1000}}
 concurrencyOverrides {maxCount=1, resolutionRule=StopFarthestThenPreventNew}
}
