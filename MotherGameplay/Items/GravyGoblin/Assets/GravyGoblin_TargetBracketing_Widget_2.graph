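# GravyGoblin HUD targeting reticle that frames a Target actor with four corner brackets.
# Activate starts a looping timer running Update, which repositions the widget from the target's bounds and deactivates when the target is invalid.
# Simulated physics size sets bracket spread and curve-driven scale pulsing, with pickups and projectiles treated as Small, while building resource type selects tint.
# Connect animates brackets, reticle rotation and corner bursts; Reset_Anim restores defaults.
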WidgetBlueprint /MotherGameplay/Items/GravyGoblin/Assets/GravyGoblin_TargetBracketing_Widget_2
animations [
 Reset_Anim {
  binding_count=9
  length_seconds=0.2
  object=&self:Reset_Anim
  sequence {
   bindings [
    TargetBracket_BottomLeft {
     id=#1
     possessed_class=Image
     tracks [
      {
       type=MovieScene2DTransformTrack
       sections [{
         type=MovieScene2DTransformSection
         channels [
          Translation.X {type=MovieSceneScriptingFloatChannel, const [0,1]}
          Translation.Y {type=MovieSceneScriptingFloatChannel, const [0,1]}
          Angle {type=MovieSceneScriptingFloatChannel, const [90,1]}
          Scale.X {type=MovieSceneScriptingFloatChannel, const [1,1]}
          Scale.Y {type=MovieSceneScriptingFloatChannel, const [1,1]}
         ]
        }]
      }
      {
       type=MovieSceneWidgetMaterialTrack
       sections [{
         type=MovieSceneParameterSection
         channels [
          Glow_Alpha {type=MovieSceneScriptingFloatChannel, const [0.5,1]}
          Pulse {type=MovieSceneScriptingFloatChannel, const [0.5,1]}
          Fill_Alpha {type=MovieSceneScriptingFloatChannel, const [0.5,1]}
         ]
        }]
      }
     ]
    }
    Reticle {
     id=#2
     possessed_class=Overlay
     tracks [
      {
       type=MovieSceneFloatTrack
       display_name=Render Opacity
       sections [{type=MovieSceneFloatSection, channels [None {type=MovieSceneScriptingFloatChannel, const [1,1]}]}]
      }
      {
       type=MovieScene2DTransformTrack
       sections [{
         type=MovieScene2DTransformSection
         channels [
          Angle {type=MovieSceneScriptingFloatChannel, const [0,1]}
          Scale.X {type=MovieSceneScriptingFloatChannel, const [1,1]}
          Scale.Y {type=MovieSceneScriptingFloatChannel, const [1,1]}
         ]
        }]
      }
     ]
    }
    Burst_TopRight {
     id=#3
     possessed_class=Image
     tracks [
      {
       type=MovieScene2DTransformTrack
       sections [{
         type=MovieScene2DTransformSection
         channels [
          Translation.X {type=MovieSceneScriptingFloatChannel, const [0,1]}
          Translation.Y {type=MovieSceneScriptingFloatChannel, const [0,1]}
         ]
        }]
      }
      {
       type=MovieSceneWidgetMaterialTrack
       sections [{type=MovieSceneParameterSection, channels [Final_Alpha {type=MovieSceneScriptingFloatChannel, const [0,1]}]}]
      }
     ]
    }
    TargetBracket_TopRight {
     id=#4
     possessed_class=Image
     tracks [
      {
       type=MovieScene2DTransformTrack
       sections [{
         type=MovieScene2DTransformSection
         channels [
          Translation.X {type=MovieSceneScriptingFloatChannel, const [0,1]}
          Translation.Y {type=MovieSceneScriptingFloatChannel, const [0,1]}
          Angle {type=MovieSceneScriptingFloatChannel, const [0,1]}
          Scale.X {type=MovieSceneScriptingFloatChannel, const [1,1]}
          Scale.Y {type=MovieSceneScriptingFloatChannel, const [1,1]}
         ]
        }]
      }
      {
       type=MovieSceneWidgetMaterialTrack
       sections [{
         type=MovieSceneParameterSection
         channels [
          Glow_Alpha {type=MovieSceneScriptingFloatChannel, const [0.5,1]}
          Pulse {type=MovieSceneScriptingFloatChannel, const [0.5,1]}
          Fill_Alpha {type=MovieSceneScriptingFloatChannel, const [0.5,1]}
         ]
        }]
      }
     ]
    }
    Burst_TopLeft {
     id=#5
     possessed_class=Image
     tracks [
      {
       type=MovieSceneWidgetMaterialTrack
       sections [{type=MovieSceneParameterSection, channels [Final_Alpha {type=MovieSceneScriptingFloatChannel, const [0,1]}]}]
      }
      {
       type=MovieScene2DTransformTrack
       sections [{
         type=MovieScene2DTransformSection
         channels [
          Translation.X {type=MovieSceneScriptingFloatChannel, const [0,1]}
          Translation.Y {type=MovieSceneScriptingFloatChannel, const [0,1]}
         ]
        }]
      }
     ]
    }
    TargetBracket_TopLeft {
     id=#6
     possessed_class=Image
     tracks [
      {
       type=MovieScene2DTransformTrack
       sections [{
         type=MovieScene2DTransformSection
         channels [
          Translation.X {type=MovieSceneScriptingFloatChannel, const [0,1]}
          Translation.Y {type=MovieSceneScriptingFloatChannel, const [0,1]}
          Angle {type=MovieSceneScriptingFloatChannel, const [0,1]}
          Scale.X {type=MovieSceneScriptingFloatChannel, const [1,1]}
          Scale.Y {type=MovieSceneScriptingFloatChannel, const [1,1]}
         ]
        }]
      }
      {
       type=MovieSceneWidgetMaterialTrack
       sections [{
         type=MovieSceneParameterSection
         channels [
          Glow_Alpha {type=MovieSceneScriptingFloatChannel, const [0.5,1]}
          Pulse {type=MovieSceneScriptingFloatChannel, const [0.5,1]}
          Fill_Alpha {type=MovieSceneScriptingFloatChannel, const [0.5,1]}
         ]
        }]
      }
     ]
    }
    Burst_BottomLeft {
     id=#7
     possessed_class=Image
     tracks [
      {
       type=MovieScene2DTransformTrack
       sections [{
         type=MovieScene2DTransformSection
         channels [
          Translation.X {type=MovieSceneScriptingFloatChannel, const [0,1]}
          Translation.Y {type=MovieSceneScriptingFloatChannel, const [0,1]}
         ]
        }]
      }
      {
       type=MovieSceneWidgetMaterialTrack
       sections [{type=MovieSceneParameterSection, channels [Final_Alpha {type=MovieSceneScriptingFloatChannel, const [0,1]}]}]
      }
     ]
    }
    Burst_BottomRight {
     id=#8
     possessed_class=Image
     tracks [
      {
       type=MovieSceneWidgetMaterialTrack
       sections [{type=MovieSceneParameterSection, channels [Final_Alpha {type=MovieSceneScriptingFloatChannel, const [0,1]}]}]
      }
      {
       type=MovieScene2DTransformTrack
       sections [{
         type=MovieScene2DTransformSection
         channels [
          Translation.X {type=MovieSceneScriptingFloatChannel, const [0,1]}
          Translation.Y {type=MovieSceneScriptingFloatChannel, const [0,1]}
         ]
        }]
      }
     ]
    }
    TargetBracket_BottomRight {
     id=#9
     possessed_class=Image
     tracks [
      {
       type=MovieScene2DTransformTrack
       sections [{
         type=MovieScene2DTransformSection
         channels [
          Translation.X {type=MovieSceneScriptingFloatChannel, const [0,1]}
          Translation.Y {type=MovieSceneScriptingFloatChannel, const [0,1]}
          Angle {type=MovieSceneScriptingFloatChannel, const [90,1]}
          Scale.X {type=MovieSceneScriptingFloatChannel, const [1,1]}
          Scale.Y {type=MovieSceneScriptingFloatChannel, const [1,1]}
         ]
        }]
      }
      {
       type=MovieSceneWidgetMaterialTrack
       sections [{
         type=MovieSceneParameterSection
         channels [
          Glow_Alpha {type=MovieSceneScriptingFloatChannel, const [0.5,1]}
          Pulse {type=MovieSceneScriptingFloatChannel, const [0.5,1]}
          Fill_Alpha {type=MovieSceneScriptingFloatChannel, const [0.5,1]}
         ]
        }]
      }
     ]
    }
   ]
   display_rate {numerator=60}
  }
 }
 Connect {
  binding_count=9
  length_seconds=1
  object=&self:Connect
  sequence {
   bindings [
    TargetBracket_BottomLeft {
     id=#1
     possessed_class=Image
     tracks [
      {
       type=MovieScene2DTransformTrack
       sections [{
         type=MovieScene2DTransformSection
         channels [
          Translation.X {type=MovieSceneScriptingFloatChannel, keys [[0,0],[12,-4]]}
          Translation.Y {type=MovieSceneScriptingFloatChannel, keys [[0,0],[12,-4]]}
          Angle {type=MovieSceneScriptingFloatChannel, const [90,1]}
          Scale.X {type=MovieSceneScriptingFloatChannel, const [1,1]}
          Scale.Y {type=MovieSceneScriptingFloatChannel, const [1,1]}
         ]
        }]
      }
      {
       type=MovieSceneWidgetMaterialTrack
       sections [{
         type=MovieSceneParameterSection
         channels [
          Glow_Alpha {type=MovieSceneScriptingFloatChannel, keys [[0,0.5],[12,1]]}
          Pulse {type=MovieSceneScriptingFloatChannel, keys [[0,0.5],[12,1]]}
          Fill_Alpha {type=MovieSceneScriptingFloatChannel, keys [[0,0.5],[12,0.75]]}
         ]
        }]
      }
     ]
    }
    Reticle {
     id=#2
     possessed_class=Overlay
     tracks [
      {
       type=MovieSceneFloatTrack
       display_name=Render Opacity
       sections [{type=MovieSceneFloatSection, channels [None {type=MovieSceneScriptingFloatChannel, const [1,1]}]}]
      }
      {
       type=MovieScene2DTransformTrack
       sections [{
         type=MovieScene2DTransformSection
         channels [
          Angle {type=MovieSceneScriptingFloatChannel, keys [[0,0],[14,45]]}
          Scale.X {type=MovieSceneScriptingFloatChannel, keys [[0,1],[14,0.9],[45,0.9],[60,0.75]]}
          Scale.Y {type=MovieSceneScriptingFloatChannel, keys [[0,1],[14,0.9],[45,0.9],[60,0.75]]}
         ]
        }]
      }
     ]
    }
    Burst_TopRight {
     id=#3
     possessed_class=Image
     tracks [
      {
       type=MovieScene2DTransformTrack
       sections [{
         type=MovieScene2DTransformSection
         channels [
          Translation.X {type=MovieSceneScriptingFloatChannel, keys [[13,0],[33,24]]}
          Translation.Y {type=MovieSceneScriptingFloatChannel, keys [[13,0],[33,-24]]}
         ]
        }]
      }
      {
       type=MovieSceneWidgetMaterialTrack
       sections [{type=MovieSceneParameterSection, channels [Final_Alpha {type=MovieSceneScriptingFloatChannel, keys [[13,0],[14,1],[33,0]]}]}]
      }
     ]
    }
    TargetBracket_TopRight {
     id=#4
     possessed_class=Image
     tracks [
      {
       type=MovieScene2DTransformTrack
       sections [{
         type=MovieScene2DTransformSection
         channels [
          Translation.X {type=MovieSceneScriptingFloatChannel, keys [[0,0],[12,-4]]}
          Translation.Y {type=MovieSceneScriptingFloatChannel, keys [[0,0],[12,4]]}
          Angle {type=MovieSceneScriptingFloatChannel, const [0,1]}
          Scale.X {type=MovieSceneScriptingFloatChannel, const [1,1]}
          Scale.Y {type=MovieSceneScriptingFloatChannel, const [1,1]}
         ]
        }]
      }
      {
       type=MovieSceneWidgetMaterialTrack
       sections [{
         type=MovieSceneParameterSection
         channels [
          Glow_Alpha {type=MovieSceneScriptingFloatChannel, keys [[0,0.5],[12,1]]}
          Pulse {type=MovieSceneScriptingFloatChannel, keys [[0,0.5],[12,1]]}
          Fill_Alpha {type=MovieSceneScriptingFloatChannel, keys [[0,0.5],[12,0.75]]}
         ]
        }]
      }
     ]
    }
    Burst_TopLeft {
     id=#5
     possessed_class=Image
     tracks [
      {
       type=MovieSceneWidgetMaterialTrack
       sections [{type=MovieSceneParameterSection, channels [Final_Alpha {type=MovieSceneScriptingFloatChannel, keys [[13,0],[14,1],[33,0]]}]}]
      }
      {
       type=MovieScene2DTransformTrack
       sections [{
         type=MovieScene2DTransformSection
         channels [
          Translation.X {type=MovieSceneScriptingFloatChannel, keys [[13,0],[33,24]]}
          Translation.Y {type=MovieSceneScriptingFloatChannel, keys [[13,0],[33,-24]]}
         ]
        }]
      }
     ]
    }
    TargetBracket_TopLeft {
     id=#6
     possessed_class=Image
     tracks [
      {
       type=MovieScene2DTransformTrack
       sections [{
         type=MovieScene2DTransformSection
         channels [
          Translation.X {type=MovieSceneScriptingFloatChannel, keys [[0,0],[12,-4]]}
          Translation.Y {type=MovieSceneScriptingFloatChannel, keys [[0,0],[12,4]]}
          Angle {type=MovieSceneScriptingFloatChannel, const [0,1]}
          Scale.X {type=MovieSceneScriptingFloatChannel, const [1,1]}
          Scale.Y {type=MovieSceneScriptingFloatChannel, const [1,1]}
         ]
        }]
      }
      {
       type=MovieSceneWidgetMaterialTrack
       sections [{
         type=MovieSceneParameterSection
         channels [
          Glow_Alpha {type=MovieSceneScriptingFloatChannel, keys [[0,0.5],[12,1]]}
          Pulse {type=MovieSceneScriptingFloatChannel, keys [[0,0.5],[12,1]]}
          Fill_Alpha {type=MovieSceneScriptingFloatChannel, keys [[0,0.5],[12,0.75]]}
         ]
        }]
      }
     ]
    }
    Burst_BottomLeft {
     id=#7
     possessed_class=Image
     tracks [
      {
       type=MovieScene2DTransformTrack
       sections [{
         type=MovieScene2DTransformSection
         channels [
          Translation.X {type=MovieSceneScriptingFloatChannel, keys [[13,0],[33,24]]}
          Translation.Y {type=MovieSceneScriptingFloatChannel, keys [[13,0],[33,24]]}
         ]
        }]
      }
      {
       type=MovieSceneWidgetMaterialTrack
       sections [{type=MovieSceneParameterSection, channels [Final_Alpha {type=MovieSceneScriptingFloatChannel, keys [[13,0],[14,1],[33,0]]}]}]
      }
     ]
    }
    Burst_BottomRight {
     id=#8
     possessed_class=Image
     tracks [
      {
       type=MovieSceneWidgetMaterialTrack
       sections [{type=MovieSceneParameterSection, channels [Final_Alpha {type=MovieSceneScriptingFloatChannel, keys [[13,0],[14,1],[33,0]]}]}]
      }
      {
       type=MovieScene2DTransformTrack
       sections [{
         type=MovieScene2DTransformSection
         channels [
          Translation.X {type=MovieSceneScriptingFloatChannel, keys [[13,0],[33,24]]}
          Translation.Y {type=MovieSceneScriptingFloatChannel, keys [[13,0],[33,24]]}
         ]
        }]
      }
     ]
    }
    TargetBracket_BottomRight {
     id=#9
     possessed_class=Image
     tracks [
      {
       type=MovieScene2DTransformTrack
       sections [{
         type=MovieScene2DTransformSection
         channels [
          Translation.X {type=MovieSceneScriptingFloatChannel, keys [[0,0],[12,-4]]}
          Translation.Y {type=MovieSceneScriptingFloatChannel, keys [[0,0],[12,-4]]}
          Angle {type=MovieSceneScriptingFloatChannel, const [90,1]}
          Scale.X {type=MovieSceneScriptingFloatChannel, const [1,1]}
          Scale.Y {type=MovieSceneScriptingFloatChannel, const [1,1]}
         ]
        }]
      }
      {
       type=MovieSceneWidgetMaterialTrack
       sections [{
         type=MovieSceneParameterSection
         channels [
          Glow_Alpha {type=MovieSceneScriptingFloatChannel, keys [[0,0.5],[12,1]]}
          Pulse {type=MovieSceneScriptingFloatChannel, keys [[0,0.5],[12,1]]}
          Fill_Alpha {type=MovieSceneScriptingFloatChannel, keys [[0,0.5],[12,0.75]]}
         ]
        }]
      }
     ]
    }
   ]
   display_rate {numerator=60}
  }
 }
]
graphs [
 EventGraph {nodes [
   n1 = Update() {is=title K2Node_CustomEvent}
   > SetVisibility(self=Get Reticle() -> (Reticle:Overlay Object Reference), InVisibility=Visible)
   n2 = Get Target() -> (Target:Actor Object Reference)
   > n3 = Is Valid(InputObject=@n2)
   Deactivate(execute=@n3.Is Not Valid)
   n4 = BroadcastPlayerImpactAtLocation!(
    Player:Player Controller Object Reference=GetOwningPlayer()
    Position:"Vector (by ref)"=Get Actor Bounds(self=@n2, bOnlyCollidingComponents=true)
    bRelativePosition:bool=false
   ) -> (Loudness:float, Duration:float, ReturnValue:bool)
   SetPositionInViewport!(
    execute=@n3.Is Valid
    self:User Widget Object Reference
    Position:Vector 2D Structure=MakeVector2D(X=@n4, Y=@n4.Duration)
    bRemoveDPIScale:bool=true
   ) -> (then)
   > n5 = Get Target Size!(execute:Exec, PhysicsObject:Actor Object Reference=Get Target()) -> (then:Exec, ReturnValue:EFortPhysicsSimSize Enum=Large) {is=title K2Node_MacroInstance}
   > Set Openness(Openness:double=Select!(
     Small:double=-15
     Average:double=-7.5
     Medium:double=4
     Large:double=19
     Huge:double=0
     Giant:double=0
     Enormous:double=0
     Invalid:double=0
     Index:EFortPhysicsSimSize Enum=@n5
    ) -> (ReturnValue:double)) -> (Output_Get:double)
   n6 = GetFloatValue!(
    self:Curve Float Object Reference=Get ScaleCurve() -> (ScaleCurve:Curve Float Object Reference)
    InTime:float=float * float(A=GetGameTimeInSeconds(), B=Select(Small=3.6, Average=1.65, Medium=0.75, Large=0.25, Huge=0, Giant=0, Enormous=0, Invalid=0, Index=@n5))
   ) -> (ReturnValue:float)
   n7 = Select(Small=0.15, Average=0.25, Medium=0.3, Large=0.325, Huge=0, Giant=0, Enormous=0, Invalid=0, Index=@n5)
   > Set BracketScale(BracketScale:double=float + float(A=Select(Small=0.6, Average=0.8, Medium=0.95, Large=1.05, Huge=0, Giant=0, Enormous=0, Invalid=0, Index=@n5), B=float * float(A=@n6, B=@n7)) {is=title K2Node_CommutativeAssociativeBinaryOperator}) -> (Output_Get:double)
   > n8 = Sequence()
   n9 = Get BracketScale() -> (BracketScale:double)
   SetRenderScale!(
    execute=@n8.then_1
    self:Widget Object Reference=[
     Get TargetBracket_TopLeft() -> (TargetBracket_TopLeft:Image Object Reference)
     Get TargetBracket_TopRight() -> (TargetBracket_TopRight:Image Object Reference)
     Get TargetBracket_BottomRight() -> (TargetBracket_BottomRight:Image Object Reference)
     Get TargetBracket_BottomLeft() -> (TargetBracket_BottomLeft:Image Object Reference)
    ]
    Scale_X:double=@n9
    Scale_Y:double=@n9
   ) -> (then)
   > SetRenderTranslation!(
    execute
    self=[Get TargetBracket_TopLeft(), Get TargetBracket_TopRight()]
    Translation_X:double=Get Openness() -> (Openness:double)
    Translation_Y:double=float * float(A=Get Openness(), B=-1)
   ) -> (then)
   n10 = Get Openness()
   > SetRenderTranslation(self=[Get TargetBracket_BottomRight(), Get TargetBracket_BottomLeft()], Translation_X=@n10, Translation_Y=@n10)
   Activate() {is=title K2Node_CustomEvent}
   > n11 = Branch(Condition=Get Activated() -> (Activated:bool))
   n12 = Set Timer by Event(execute=@n11.else, Delegate=@n1, Time=0.01, bLooping=true)
   > Set TickTimer(TickTimer:Timer Handle Structure=@n12) -> (Output_Get:Timer Handle Structure)
   > Set Activated(Activated:bool=true) -> (Output_Get:bool)
   Deactivate() {is=title K2Node_CustomEvent}
   > Branch(Condition=Get Activated())
   > SetVisibility(self=Get Reticle(), InVisibility=Hidden)
   > Clear and Invalidate Timer by Handle(Handle=Get TickTimer() -> (TickTimer:Timer Handle Structure))
   > Set Activated(Activated=false)
   n13 = Cast To BuildingSMActor!(execute, Object=Get Target()) -> (then, CastFailed, AsBuilding Static Mesh:Building Static Mesh Object Reference) {is=title K2Node_DynamicCast}
   > SetColorAndOpacity(
    self:Image Object Reference=[Get TargetBracket_TopLeft(), Get TargetBracket_TopRight(), Get TargetBracket_BottomRight(), Get TargetBracket_BottomLeft()]
    InColorAndOpacity:Linear Color Structure=Select!(
     Wood:Linear Color Structure=(R=0.6939,G=1,B=0.6038,A=1)
     Stone:Linear Color Structure=(R=0.6296,G=0.9177,B=1,A=1)
     Metal:Linear Color Structure=(R=0.8632,G=0.6038,B=1,A=1)
     Permanite:Linear Color Structure=(R=0.6939,G=1,B=0.6038,A=1)
     GoldCurrency:Linear Color Structure=(R=0.6939,G=1,B=0.6038,A=1)
     Ingredient:Linear Color Structure=(R=0.6939,G=1,B=0.6038,A=1)
     None:Linear Color Structure=(R=0.6939,G=1,B=0.6038,A=1)
     Index:EFortResourceType Enum=Get ResourceType!(self:Building Static Mesh Object Reference=@n13) -> (ResourceType:EFortResourceType Enum=Wood)
    ) -> (ReturnValue:Linear Color Structure)
   )
   SetColorAndOpacity(
    execute=@n13.CastFailed
    self=[Get TargetBracket_BottomLeft(), Get TargetBracket_TopLeft(), Get TargetBracket_TopRight(), Get TargetBracket_BottomRight()]
    InColorAndOpacity=(R=0.8632,G=0.6038,B=1,A=1)
   )
   Event Construct()
   n14 = Get BracketScale()
   > SetRenderScale(
    self=[Get TargetBracket_TopLeft(), Get TargetBracket_TopRight(), Get TargetBracket_BottomRight(), Get TargetBracket_BottomLeft()]
    Scale_X=@n14
    Scale_Y=@n14
   )
   > SetVisibility(self=Get Reticle(), InVisibility=Hidden)
   ResetAnim()
   Get ScaleVariance() -> (ScaleVariance:double)
  ]}
 GetTargetSize {nodes [
   n1 = Inputs!() -> (execute, PhysicsObject:Actor Object Reference)
   > n2 = GetPhysicsObjectSimSize!(execute, PhysicsObject:Actor Object Reference=@n1) -> (then, ReturnValue:EFortPhysicsSimSize Enum=Small)
   n3 = Cast To FortPickup!(Object=@n1) -> (AsFort Pickup:Fort Pickup Object Reference, bSuccess:bool=~) {is=title K2Node_DynamicCast}
   n4 = Cast To FortProjectileBase!(Object=@n1) -> (AsFort Projectile Base:Fort Projectile Base Object Reference, bSuccess:bool=~) {is=title K2Node_DynamicCast}
   > Outputs!(
    then:Exec
    ReturnValue:EFortPhysicsSimSize Enum=Select!(Option 0:EFortPhysicsSimSize Enum=@n2, Option 1:EFortPhysicsSimSize Enum=Small, Index:bool=OR Boolean(A=@n3.bSuccess, B=@n4.bSuccess)) -> (ReturnValue:EFortPhysicsSimSize Enum)
   )
  ]}
 ResetAnim {nodes [
   ResetAnim() {is=title K2Node_FunctionEntry}
   > PlayAnimation(InAnimation=Get Reset_Anim() -> (Reset_Anim:Widget Animation Object Reference))
  ]}
]
variables [
 Activated {type=boolean}
 BracketScale {type=number, default=1}
 BracketScalePerTick {type=number, default=0.1}
 OffsetX {type=number}
 OffsetY {type=number}
 Openness {type=number, default=10}
 ScaleCurve {type=CurveFloat, default=&/GravyGoblin_ReticleAnimationCurve}
 ScaleSizeUp {type=boolean}
 ScaleVariance {type=number, default=0.3}
 Target {type=Actor, default=null}
 TickTimer {type=TimerHandle}
]
widgets {
 info {inheritedWidgetCount=1, parentClass=&FortHUDElementWidget, rootWidgetClass=&Overlay, widgetCount=18}
 widgets [
  Reticle {widgetClassPath=&Overlay, bIsVariable=true, widget=Reticle, properties {renderTransformPivot=<0 0>}}
  All {
   widgetClassPath=&SizeBox
   parent=Reticle
   widget=All
   properties {bOverride_HeightOverride=true, bOverride_WidthOverride=true, heightOverride=96, widthOverride=96}
   slot {class=OverlaySlot, properties {horizontalAlignment=HAlign_Fill, verticalAlignment=VAlign_Fill}}
  }
  Overlay_1 {
   widgetClassPath=&Overlay
   parent=All
   widget=Overlay_1
   properties {renderTransform {angle=0, scale=<1 1>, shear=<0 0>, translation=<-48 -48>}}
   slot {class=SizeBoxSlot}
  }
  HorizontalBox_0 {
   widgetClassPath=&HorizontalBox
   parent=Overlay_1
   widget=HorizontalBox_0
   slot {class=OverlaySlot, properties {horizontalAlignment=HAlign_Fill, verticalAlignment=VAlign_Fill}}
  }
  VerticalBox_0 {
   widgetClassPath=&VerticalBox
   parent=HorizontalBox_0
   widget=VerticalBox_0
   properties {renderTransform {angle=0, scale=<-1 1>, shear=<0 0>, translation=<0 0>}}
   slot {class=HorizontalBoxSlot, properties {size {sizeRule=Fill, value=1}}}
  }
  TargetBracket_TopLeft {
   widgetClassPath=&Image
   bIsVariable=true
   parent=VerticalBox_0
   widget=TargetBracket_TopLeft
   properties {
    visibility=SelfHitTestInvisible
    brush {
     bIsDynamicallyLoaded=false
     drawAs=Image
     imageType=NoImage
     mirroring=NoMirror
     resourceName=None
     resourceObject=&/M_GravyGoblin_Bracket_Inst
     tiling=NoTile
     imageSize=<32 32>
     margin=<ltrb 0 0 0 0>
     outlineSettings {
      bUseBrushTransparency=false
      roundingType=HalfHeightRadius
      width=0
      color {colorUseRule=UseColor_Specified, specifiedColor=<rgba 0 0 0 0>}
      cornerRadii=<0 0 0 0>
     }
     tintColor {colorUseRule=UseColor_Specified, specifiedColor=<rgba 1 1 1 1>}
     uVRegion {bIsValid=false, max=<0 0>, min=<0 0>}
    }
   }
   slot {class=VerticalBoxSlot, properties {size {sizeRule=Fill, value=1}}}
  }
  TargetBracket_BottomLeft {
   widgetClassPath=&Image
   bIsVariable=true
   parent=VerticalBox_0
   widget=TargetBracket_BottomLeft
   properties {
    visibility=SelfHitTestInvisible
    brush {
     bIsDynamicallyLoaded=false
     drawAs=Image
     imageType=NoImage
     mirroring=NoMirror
     resourceName=None
     resourceObject=&M_GravyGoblin_Bracket_Inst
     tiling=NoTile
     imageSize=<32 32>
     margin=<ltrb 0 0 0 0>
     outlineSettings {
      bUseBrushTransparency=false
      roundingType=HalfHeightRadius
      width=0
      color {colorUseRule=UseColor_Specified, specifiedColor=<rgba 0 0 0 0>}
      cornerRadii=<0 0 0 0>
     }
     tintColor {colorUseRule=UseColor_Specified, specifiedColor=<rgba 1 1 1 1>}
     uVRegion {bIsValid=false, max=<0 0>, min=<0 0>}
    }
    renderTransform {angle=90, scale=<1 1>, shear=<0 0>, translation=<0 0>}
   }
   slot {class=VerticalBoxSlot, properties {size {sizeRule=Fill, value=1}}}
  }
  VerticalBox {
   widgetClassPath=&VerticalBox
   parent=HorizontalBox_0
   widget=VerticalBox
   slot {class=HorizontalBoxSlot, properties {size {sizeRule=Fill, value=1}}}
  }
  TargetBracket_TopRight {
   widgetClassPath=&Image
   bIsVariable=true
   parent=VerticalBox
   widget=TargetBracket_TopRight
   properties {
    visibility=SelfHitTestInvisible
    brush {
     bIsDynamicallyLoaded=false
     drawAs=Image
     imageType=NoImage
     mirroring=NoMirror
     resourceName=None
     resourceObject=&M_GravyGoblin_Bracket_Inst
     tiling=NoTile
     imageSize=<32 32>
     margin=<ltrb 0 0 0 0>
     outlineSettings {
      bUseBrushTransparency=false
      roundingType=HalfHeightRadius
      width=0
      color {colorUseRule=UseColor_Specified, specifiedColor=<rgba 0 0 0 0>}
      cornerRadii=<0 0 0 0>
     }
     tintColor {colorUseRule=UseColor_Specified, specifiedColor=<rgba 1 1 1 1>}
     uVRegion {bIsValid=false, max=<0 0>, min=<0 0>}
    }
   }
   slot {class=VerticalBoxSlot, properties {size {sizeRule=Fill, value=1}}}
  }
  TargetBracket_BottomRight {
   widgetClassPath=&Image
   bIsVariable=true
   parent=VerticalBox
   widget=TargetBracket_BottomRight
   properties {
    visibility=SelfHitTestInvisible
    brush {
     bIsDynamicallyLoaded=false
     drawAs=Image
     imageType=NoImage
     mirroring=NoMirror
     resourceName=None
     resourceObject=&M_GravyGoblin_Bracket_Inst
     tiling=NoTile
     imageSize=<32 32>
     margin=<ltrb 0 0 0 0>
     outlineSettings {
      bUseBrushTransparency=false
      roundingType=HalfHeightRadius
      width=0
      color {colorUseRule=UseColor_Specified, specifiedColor=<rgba 0 0 0 0>}
      cornerRadii=<0 0 0 0>
     }
     tintColor {colorUseRule=UseColor_Specified, specifiedColor=<rgba 1 1 1 1>}
     uVRegion {bIsValid=false, max=<0 0>, min=<0 0>}
    }
    renderTransform {angle=90, scale=<1 1>, shear=<0 0>, translation=<0 0>}
   }
   slot {class=VerticalBoxSlot, properties {size {sizeRule=Fill, value=1}}}
  }
  Burst {
   widgetClassPath=&HorizontalBox
   parent=Overlay_1
   widget=Burst
   slot {class=OverlaySlot, properties {horizontalAlignment=HAlign_Fill, verticalAlignment=VAlign_Fill}}
  }
  VerticalBox_1 {
   widgetClassPath=&VerticalBox
   parent=Burst
   widget=VerticalBox_1
   properties {renderTransform {angle=0, scale=<-1 1>, shear=<0 0>, translation=<0 0>}}
   slot {class=HorizontalBoxSlot, properties {size {sizeRule=Fill, value=1}}}
  }
  Burst_TopLeft {
   widgetClassPath=&Image
   bIsVariable=true
   parent=VerticalBox_1
   widget=Burst_TopLeft
   properties {
    visibility=SelfHitTestInvisible
    brush {
     bIsDynamicallyLoaded=false
     drawAs=Image
     imageType=NoImage
     mirroring=NoMirror
     resourceName=None
     resourceObject=&/M_GravyGoblin_Bracket_Burst
     tiling=NoTile
     imageSize=<32 32>
     margin=<ltrb 0 0 0 0>
     outlineSettings {
      bUseBrushTransparency=false
      roundingType=HalfHeightRadius
      width=0
      color {colorUseRule=UseColor_Specified, specifiedColor=<rgba 0 0 0 0>}
      cornerRadii=<0 0 0 0>
     }
     tintColor {colorUseRule=UseColor_Specified, specifiedColor=<rgba 1 1 1 1>}
     uVRegion {bIsValid=false, max=<0 0>, min=<0 0>}
    }
   }
   slot {class=VerticalBoxSlot, properties {size {sizeRule=Fill, value=1}}}
  }
  Burst_BottomLeft {
   widgetClassPath=&Image
   bIsVariable=true
   parent=VerticalBox_1
   widget=Burst_BottomLeft
   properties {
    visibility=SelfHitTestInvisible
    brush {
     bIsDynamicallyLoaded=false
     drawAs=Image
     imageType=NoImage
     mirroring=NoMirror
     resourceName=None
     resourceObject=&M_GravyGoblin_Bracket_Burst
     tiling=NoTile
     imageSize=<32 32>
     margin=<ltrb 0 0 0 0>
     outlineSettings {
      bUseBrushTransparency=false
      roundingType=HalfHeightRadius
      width=0
      color {colorUseRule=UseColor_Specified, specifiedColor=<rgba 0 0 0 0>}
      cornerRadii=<0 0 0 0>
     }
     tintColor {colorUseRule=UseColor_Specified, specifiedColor=<rgba 1 1 1 1>}
     uVRegion {bIsValid=false, max=<0 0>, min=<0 0>}
    }
    renderTransform {angle=90, scale=<1 1>, shear=<0 0>, translation=<0 0>}
   }
   slot {class=VerticalBoxSlot, properties {size {sizeRule=Fill, value=1}}}
  }
  VerticalBox_2 {
   widgetClassPath=&VerticalBox
   parent=Burst
   widget=VerticalBox_2
   slot {class=HorizontalBoxSlot, properties {size {sizeRule=Fill, value=1}}}
  }
  Burst_TopRight {
   widgetClassPath=&Image
   bIsVariable=true
   parent=VerticalBox_2
   widget=Burst_TopRight
   properties {
    visibility=SelfHitTestInvisible
    brush {
     bIsDynamicallyLoaded=false
     drawAs=Image
     imageType=NoImage
     mirroring=NoMirror
     resourceName=None
     resourceObject=&M_GravyGoblin_Bracket_Burst
     tiling=NoTile
     imageSize=<32 32>
     margin=<ltrb 0 0 0 0>
     outlineSettings {
      bUseBrushTransparency=false
      roundingType=HalfHeightRadius
      width=0
      color {colorUseRule=UseColor_Specified, specifiedColor=<rgba 0 0 0 0>}
      cornerRadii=<0 0 0 0>
     }
     tintColor {colorUseRule=UseColor_Specified, specifiedColor=<rgba 1 1 1 1>}
     uVRegion {bIsValid=false, max=<0 0>, min=<0 0>}
    }
   }
   slot {class=VerticalBoxSlot, properties {size {sizeRule=Fill, value=1}}}
  }
  Burst_BottomRight {
   widgetClassPath=&Image
   bIsVariable=true
   parent=VerticalBox_2
   widget=Burst_BottomRight
   properties {
    visibility=SelfHitTestInvisible
    brush {
     bIsDynamicallyLoaded=false
     drawAs=Image
     imageType=NoImage
     mirroring=NoMirror
     resourceName=None
     resourceObject=&M_GravyGoblin_Bracket_Burst
     tiling=NoTile
     imageSize=<32 32>
     margin=<ltrb 0 0 0 0>
     outlineSettings {
      bUseBrushTransparency=false
      roundingType=HalfHeightRadius
      width=0
      color {colorUseRule=UseColor_Specified, specifiedColor=<rgba 0 0 0 0>}
      cornerRadii=<0 0 0 0>
     }
     tintColor {colorUseRule=UseColor_Specified, specifiedColor=<rgba 1 1 1 1>}
     uVRegion {bIsValid=false, max=<0 0>, min=<0 0>}
    }
    renderTransform {angle=90, scale=<1 1>, shear=<0 0>, translation=<0 0>}
   }
   slot {class=VerticalBoxSlot, properties {size {sizeRule=Fill, value=1}}}
  }
  TutorialHighlight {widgetClassPath=&TutorialHighlight}
 ]
}
