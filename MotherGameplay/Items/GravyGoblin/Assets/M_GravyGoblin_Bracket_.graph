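# Translucent UI material that draws the Gravy Goblin bracket reticle from a channel-packed mask scaled about its center.
# Red (fill) and green (outline) get adjustable edge cutoff and hardness, then combine with blue glow, each with its own alpha.
# A Pulse parameter blends the outline toward a sine-over-time version.
# The combined mask lerps falloff to reticle color and drives final alpha; a panning scanline branch stays separate from it.
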Material /MotherGameplay/Items/GravyGoblin/Assets/M_GravyGoblin_Bracket_
graph {
 properties {blendMode=BLEND_Translucent, materialDomain=MD_UI, pixelDepthOffsetMode=PDOM_Legacy, usedShadingModels=Default Lit}
 nodes [
  n1 = Panner(Coordinate=TextureCoordinate()) {speedX=0.5, speedY=0.5}
  Clamp(None=Multiply(A=TextureSampleParameter2D(UVs=@n1) {parameterName=Panner, texture=/Game/Global/Textures/Default/Masks_and_Gradients/Scanlines_Offset}.RGB) {constB=3}) {maxDefault=5, minDefault=0.3}
  n2 = TextureSampleParameter2D(UVs=MaterialFunctionCall(UVs=TextureCoordinate(), Texture Scale=ScalarParameter() {defaultValue=1, parameterName=CenterSize}) {materialFunction=&ScaleUVsByCenter}.UVs) {
   parameterName=Reticle
   samplerSource=SSM_Clamp_WorldGroupSettings
   samplerType=SAMPLERTYPE_LinearColor
   texture=&/T_GravyGoblin_Bracket_Mask_
  }
  n3 = ScalarParameter() {defaultValue=0.5, parameterName=Glow_Alpha}
  n4 = ScalarParameter() {defaultValue=0.05, parameterName=Fill_EdgeSubtract}
  n5 = Saturate(None=Multiply(A=Subtract(A=@n2.R, B=@n4), B=ScalarParameter() {defaultValue=1.25, parameterName=Fill_Edge_Hardness}))
  n6 = ScalarParameter() {defaultValue=0.5, parameterName=Fill_Alpha}
  n7 = ScalarParameter() {defaultValue=0.05, parameterName=Outline_EdgeSubtract}
  n8 = Saturate(None=Multiply(A=Subtract(A=@n2.G, B=@n7), B=ScalarParameter() {defaultValue=1.25, parameterName=Outline_Edge_Hardness}))
  n9 = Multiply(A=@n8, B=ScalarParameter() {defaultValue=0.5, parameterName=Outline_Alpha})
  n10 = ScalarParameter() {defaultValue=0.5, parameterName=TimeSine-Min}
  n11 = ScalarParameter() {defaultValue=1, parameterName=TimeSine-Max}
  n12 = Time()
  n13 = ScalarParameter() {defaultValue=2, parameterName=TimeSpeed}
  n14 = Add(A=Sine(None=Multiply(A=@n12, B=@n13)))
  n15 = DotProduct(A=@n8, B=Saturate(None=LinearInterpolate(A=@n10, B=@n11, Alpha=@n14)))
  n16 = Add(
   A=Saturate(None=Multiply(A=@n5, B=@n6))
   B=LinearInterpolate(A=Saturate(None=@n9), B=Saturate(None=@n15), Alpha=ScalarParameter() {parameterName=Pulse})
  )
  n17 = Saturate(None=Add(A=Multiply(A=@n2.B, B=@n3), B=Saturate(None=@n16)))
  LinearInterpolate(
   A=VectorParameter() {parameterName=Falloff_Color, defaultValue=<rgba 0 0.2462 1 1>}.RGB
   B=VectorParameter() {parameterName=Reticle Color, defaultValue=<rgba 0 1 1 1>}.RGB
   Alpha=@n17
  )
  Saturate(None=Multiply(A=@n17, B=ScalarParameter() {defaultValue=1, parameterName=Final_Alpha}))
 ]
}
